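# Control Rig for the large female purple Bunny player body, adding procedural corrections to deform, dynamic and accessory bones.
# After a shared DeformRig function, spherical pose readers on the right thigh move the front frill, one interpolated and scaled by a crouch variable.
# Two-bone IK pins both wrist_l/r chains to the front pocket.
# Head pose readers reshape collar bones, then Spherical Collision functions run between those bones and the head.
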ControlRigBlueprint /BRCosmetics/Characters/Player/Female/Large/Bodies/F_LRG_Bunny_Purple/Meshes/F_LRG_Bunny_Purple_CtrlRig
graphs [
 ContainedGraph {nodes [
   n1 = RigVMFunctionEntryNode()
   SequenceExecution_2 = &RigVMFunction_Sequence(ExecuteContext=@n1.ExecuteContext)
   SequenceExecution_2_1 = &RigVMFunction_Sequence(ExecuteContext=@SequenceExecution_2.B)
   SequenceExecution_2_2 = &RigVMFunction_Sequence(ExecuteContext=@SequenceExecution_2_1.B)
   RigVMFunctionReturnNode(A=@SequenceExecution_2.A, B=@SequenceExecution_2_1.A, C=@SequenceExecution_2_2.A, D=@SequenceExecution_2_2.B)
  ]}
 RigVMFunctionLibrary {links [], nodes []}
 RigVMFunction_Sequence_1_ContainedGraph {nodes [
   n1 = RigVMFunctionEntryNode()
   RigVMFunction_Sequence = &RigVMFunction_Sequence(ExecuteContext=@n1.ExecuteContext)
   n2 = &RigVMFunction_Sequence(ExecuteContext=@RigVMFunction_Sequence.B)
   n3 = &RigVMFunction_Sequence(ExecuteContext=@n2.B)
   RigVMFunctionReturnNode(A=@RigVMFunction_Sequence.A, B=@n2.A, C=@n3.A, D=@n3.B)
  ]}
 RigVMModel {nodes [
   RigUnit_BeginExecution = &RigUnit_BeginExecution()
   RigVMFunction_Sequence = RigVMAggregateNode(ExecuteContext=@RigUnit_BeginExecution.ExecutePin)
   FN_DeformRig_FNC = RigVMFunctionReferenceNode(
    DeformRig_ Curve="(deformRig_l_upperarm_up_alpha_72_#1=False,deformRig_r_upperarm_up_alpha_73_#2=False,deformRig_l_upperarm_fr_alpha_74_#3=False,deformRig_r_upperarm_fr_alpha_75_CB51B8 ...(truncated)"
    DeformRig_Bool="(upperarm_1_#4=True,elbow_3_#5=True,wrist_7_#6=True,groin_8_#7=True,glute_17_#8=True,knee_9_40C408324CA261 ...(truncated)"
    DeformRig_Debug=(upperarm_debug_18_#4=False,wrist_debug_20_#6=False,lat_debug_26_#9=False,groinGlute_debug_27_#7=False)
    DeformRig_Value="(upperarm_up_18_#10=(Rotation=(X=0,Y=0,Z=0,W=1),Translation=(X=0,Y=0,Z=8),Scale3D=(X=0.7,Y=1,Z=1)),upperarm_fwd_fr_19_#11=(Rotation=(X=0,Y=0,Z=0,W=1),Translation=(X=0,Y=10,Z=0),Scale3D=(X=0.5,Y=1,Z= ...(truncated)"
    ExecuteContext=@RigVMFunction_Sequence.A
   )
   > n1 = &RigUnit_SphericalPoseReader(
    ActiveRegionScaleFactors=(PositiveWidth=3.5,NegativeWidth=0.5,PositiveHeight=3,NegativeHeight=3)
    ActiveRegionSize=0.1
    Debug=(bDrawDebug=false,DebugScale=25,DebugSegments=20,DebugThickness=0.25)
    DriverAxis=(X=1,Y=0,Z=0)
    DriverItem=(Type=Bone,Name="thigh_r")
    FalloffRegionScaleFactors=(PositiveWidth=1,NegativeWidth=1,PositiveHeight=1,NegativeHeight=1)
    FalloffSize=0.45
    FlipHeightScaling=false
    FlipWidthScaling=true
    OptionalParentItem=(Type=Bone)
    RotationOffset=(X=-80,Y=15,Z=-20)
   )
   AlphaInterp = &RigVMFunction_AlphaInterp(
    InterpSpeedDecreasing=10
    InterpSpeedIncreasing=10
    OutRange=(Min=1,Max=0)
    bClampResult=true
    bMapRange=true
    Value=@n1.OutputParam
    Scale=RigVMVariableNode() {variable=crouch still}.Value
   )
   > &RigUnit_ModifyTransforms(
    ItemToModify=((Item=(Type=Bone,Name="dyn_frill_frt"),Transform=(Rotation=(X=0,Y=0,Z=0,W=1),Translation=(X=0,Y=15,Z=-35),Scale3D=(X=1,Y=1,Z=1))))
    WeightMinimum=0
    Weight=@AlphaInterp.Result
   )
   > n2 = &RigUnit_SphericalPoseReader(
    ActiveRegionScaleFactors=(PositiveWidth=3.5,NegativeWidth=0.5,PositiveHeight=3,NegativeHeight=3)
    ActiveRegionSize=0.1
    Debug=(bDrawDebug=false,DebugScale=25,DebugSegments=20,DebugThickness=0.25)
    DriverAxis=(X=1,Y=0,Z=0)
    DriverItem=(Type=Bone,Name="thigh_r")
    FalloffRegionScaleFactors=(PositiveWidth=1,NegativeWidth=1,PositiveHeight=1,NegativeHeight=1)
    FalloffSize=0.45
    FlipHeightScaling=false
    FlipWidthScaling=true
    OptionalParentItem=(Type=Bone)
    RotationOffset=(X=-90,Y=0,Z=-5)
   )
   > &RigUnit_ModifyTransforms(
    ItemToModify="((Item=(Type=Bone,Name=\"dyn_frill_frt\"),Transform=(Rotation=(X=0,Y=0,Z=0.1305,W=0.9914),Translation=(X=0,Y=0,Z=5),Scale3D=(X=1,Y=1,Z=1))),(Item=(Type=Bone,Name=\"dyn_frill_frt_r\"),Transform=(Rotation=(X=0,Y=0,Z=-0.1305,W=0.9914),Translation=(X=0,Y=-5,Z=-5), ...(truncated)"
    WeightMinimum=0
    Weight=@n2.OutputParam
   )
   n3 = /Script/ControlRig.RigUnit_ProjectTransformToNewParent(
    Child=(Type=Bone,Name="wrist_l")
    NewParent=(Type=Bone,Name="front_pocket_parent")
    OldParent=(Type=Bone,Name="front_pocket_parent")
    bChildInitial=True
    bNewParentInitial=false
    bOldParentInitial=True
   ) {is=struct RigVMUnitNode}
   /Script/ControlRig.RigUnit_TwoBoneIKSimplePerItem(
    DebugSettings=(Scale=10,WorldOffset=(Rotation=(X=0,Y=0,Z=0,W=1),Translation=(X=0,Y=0,Z=0),Scale3D=(X=1,Y=1,Z=1)))
    EffectorItem=(Type=Bone,Name="wrist_l")
    ItemA=(Type=Bone,Name="shoulder_l")
    ItemB=(Type=Bone,Name="elbow_l")
    PoleVector=(X=0,Y=0,Z=0)
    PoleVectorKind=Location
    PoleVectorSpace=(Type=Control,Name="elbow_l_ctrl_test")
    PrimaryAxis=(X=1,Y=0,Z=0)
    SecondaryAxis=(X=0,Y=1,Z=0)
    SecondaryAxisWeight=0
    StretchMaximumRatio=1.25
    StretchStartRatio=0.75
    Weight=1
    bEnableStretch=false
    bPropagateToChildren=True
    Effector=@n3.Transform
    ExecutePin=@RigVMFunction_Sequence.B
   ) {is=struct RigVMUnitNode}
   n4 = &RigUnit_ProjectTransformToNewParent(
    Child=(Type=Bone,Name="wrist_r")
    NewParent=(Type=Bone,Name="front_pocket_parent")
    OldParent=(Type=Bone,Name="front_pocket_parent")
    bChildInitial=True
    bNewParentInitial=false
    bOldParentInitial=True
   )
   > &RigUnit_TwoBoneIKSimplePerItem(
    DebugSettings=(Scale=10,WorldOffset=(Rotation=(X=0,Y=0,Z=0,W=1),Translation=(X=0,Y=0,Z=0),Scale3D=(X=1,Y=1,Z=1)))
    EffectorItem=(Type=Bone,Name="wrist_r")
    ItemA=(Type=Bone,Name="shoulder_r")
    ItemB=(Type=Bone,Name="elbow_r")
    PoleVector=(X=0,Y=0,Z=-5)
    PoleVectorKind=Location
    PoleVectorSpace=(Type=Control,Name="elbow_r_ctrl_test")
    PrimaryAxis=(X=-0.8,Y=0,Z=0)
    SecondaryAxis=(X=0,Y=1,Z=0)
    SecondaryAxisWeight=0
    StretchMaximumRatio=1.25
    StretchStartRatio=0.75
    Weight=1
    bEnableStretch=false
    bPropagateToChildren=True
    Effector=@n4.Transform
   )
   > SequenceExecution_2 = RigVMAggregateNode()
   n5 = &RigUnit_SphericalPoseReader(
    ActiveRegionScaleFactors=(PositiveWidth=2,NegativeWidth=2,PositiveHeight=2.4,NegativeHeight=0.5)
    ActiveRegionSize=0.1
    Debug=(bDrawDebug=false,DebugScale=25,DebugSegments=20,DebugThickness=0.25)
    DriverAxis=(X=0,Y=1,Z=0)
    DriverItem=(Type=Bone,Name="head")
    FalloffRegionScaleFactors=(PositiveWidth=1,NegativeWidth=1,PositiveHeight=1,NegativeHeight=0.4)
    FalloffSize=0.25
    OptionalParentItem=(Type=Bone)
    RotationOffset=(X=120,Y=0,Z=0)
    ExecutePin=@SequenceExecution_2.B
   )
   > RigVMVariableNode(Value=@n5.OutputParam) {variable=head_SPR_look_left_alpha}
   n6 = &RigUnit_SphericalPoseReader(
    ActiveRegionScaleFactors=(PositiveWidth=2,NegativeWidth=2,PositiveHeight=3,NegativeHeight=0.5)
    ActiveRegionSize=0.1
    Debug=(bDrawDebug=true,DebugScale=25,DebugSegments=20,DebugThickness=0.25)
    DriverAxis=(X=0,Y=1,Z=0)
    DriverItem=(Type=Bone,Name="head")
    FalloffRegionScaleFactors=(PositiveWidth=1,NegativeWidth=1,PositiveHeight=0,NegativeHeight=0.4)
    FalloffSize=0.27
    FlipHeightScaling=true
    FlipWidthScaling=false
    OptionalParentItem=(Type=Bone)
    RotationOffset=(X=60,Y=0,Z=0)
    ExecutePin=@SequenceExecution_2.C
   )
   > RigVMVariableNode(Value=@n6.OutputParam) {variable=head_SPR_look_right_alpha}
   n7 = &RigUnit_SphericalPoseReader(
    ActiveRegionScaleFactors=(PositiveWidth=2,NegativeWidth=0.5,PositiveHeight=2,NegativeHeight=2)
    ActiveRegionSize=0.075
    Debug=(bDrawDebug=false,DebugScale=25,DebugSegments=20,DebugThickness=0.25)
    DriverAxis=(X=0,Y=1,Z=0)
    DriverItem=(Type=Bone,Name="head")
    FalloffRegionScaleFactors=(PositiveWidth=1,NegativeWidth=0.4,PositiveHeight=1,NegativeHeight=1)
    FalloffSize=0.2
    FlipHeightScaling=false
    FlipWidthScaling=true
    OptionalParentItem=(Type=Bone)
    RotationOffset=(X=90,Y=0,Z=12.5)
    ExecutePin=@SequenceExecution_2.A
   )
   > RigVMVariableNode(Value=@n7.OutputParam) {variable=head_SPR_look_down_alpha}
   n8 = &RigUnit_SphericalPoseReader(
    ActiveRegionScaleFactors=(PositiveWidth=4,NegativeWidth=1,PositiveHeight=3,NegativeHeight=3)
    ActiveRegionSize=0.07
    Debug=(bDrawDebug=false,bDraw2D=False,bDrawLocalAxes=false,DebugScale=25,DebugSegments=20,DebugThickness=0.25)
    DriverAxis=(X=1,Y=0,Z=0)
    DriverItem=(Type=Bone,Name="head")
    FalloffRegionScaleFactors=(PositiveWidth=1,NegativeWidth=1,PositiveHeight=1,NegativeHeight=1)
    FalloffSize=0.4
    FlipHeightScaling=False
    FlipWidthScaling=False
    OptionalParentItem=(Type=Bone,Name="spine_05")
    RotationOffset=(X=70,Y=0,Z=-145)
    ExecutePin=@RigVMFunction_Sequence.C
   )
   > &RigUnit_ModifyTransforms(
    ItemToModify="((Item=(Type=Bone,Name=\"dyn_collar_back_parent\"),Transform=(Rotation=(X=0,Y=0.8192,Z=0,W=0.5736),Translation=(X=-1,Y=0,Z=-1),Scale3D=(X=0.7,Y=1,Z=1))),(Item=(Type=Bone,Name=\"dyn_collar_side_r_parent\"),Transform=(Rotation=(X=-0.1983,Y=0.6209,Z=0.1664,W=0.73 ...(truncated)"
    WeightMinimum=0
    Weight=@n8.OutputParam
   )
   > &RigUnit_ModifyTransforms(
    ItemToModify="((Item=(Type=Bone,Name=\"dyn_collar_bk_l_parent\"),Transform=(Rotation=(X=0,Y=0.3007,Z=0,W=0.9537),Translation=(X=-0.5,Y=0,Z=0),Scale3D=(X=0.8,Y=1,Z=1))),(Item=(Type=Bone,Name=\"dyn_collar_bk_r_parent\"),Transform=(Rotation=(X=0,Y=0.4226,Z=0,W=0.9063),Translat ...(truncated)"
    WeightMinimum=0
    Weight=@n8.OutputParam
   )
   > &RigUnit_ModifyTransforms(
    ItemToModify="((Item=(Type=Bone,Name=\"dyn_collar_side_l_parent\"),Transform=(Rotation=(X=0,Y=0.3827,Z=0,W=0.9239),Translation=(X=-0.5,Y=0,Z=-3),Scale3D=(X=0.7,Y=1,Z=1))),(Item=(Type=Bone,Name=\"dyn_collar_fr_l_parent\"),Transform=(Rotation=(X=0.2064,Y=0.2936,Z=-0.0651,W=0. ...(truncated)"
    WeightMinimum=0
    Weight=RigVMVariableNode() {variable=head_SPR_look_left_alpha}.Value
   )
   > &RigUnit_ModifyTransforms(
    ItemToModify="((Item=(Type=Bone,Name=\"dyn_collar_side_r_parent\"),Transform=(Rotation=(X=0,Y=0.3007,Z=0,W=0.9537),Translation=(X=-0.5,Y=0,Z=-1),Scale3D=(X=0.7,Y=1,Z=1))),(Item=(Type=Bone,Name=\"dyn_collar_fr_r_parent\"),Transform=(Rotation=(X=-0.2064,Y=0.2936,Z=0.0651,W=0. ...(truncated)"
    WeightMinimum=0
    Weight=RigVMVariableNode() {variable=head_SPR_look_right_alpha}.Value
   )
   > n9 = &RigUnit_SphericalPoseReader(
    ActiveRegionScaleFactors=(PositiveWidth=2,NegativeWidth=2,PositiveHeight=1,NegativeHeight=0.5)
    ActiveRegionSize=0.1
    Debug=(bDrawDebug=false,DebugScale=25,DebugSegments=20,DebugThickness=0.25)
    DriverAxis=(X=0,Y=1,Z=0)
    DriverItem=(Type=Bone,Name="head")
    FalloffRegionScaleFactors=(PositiveWidth=1,NegativeWidth=1,PositiveHeight=1,NegativeHeight=0.6)
    FalloffSize=0.25
    OptionalParentItem=(Type=Bone)
    RotationOffset=(X=145,Y=0,Z=0)
   )
   > &RigUnit_ModifyTransforms(
    ItemToModify="((Item=(Type=Bone,Name=\"dyn_collar_side_l_parent\"),Transform=(Rotation=(X=0,Y=0.3007,Z=0,W=0.9537),Translation=(X=-6,Y=0,Z=0),Scale3D=(X=0.5,Y=1,Z=1))),(Item=(Type=Bone,Name=\"dyn_collar_fr_l_parent\"),Transform=(Rotation=(X=0,Y=0.2164,Z=0,W=0.9763),Translat ...(truncated)"
    WeightMinimum=0
    Weight=@n9.OutputParam
   )
   > n10 = &RigUnit_SphericalPoseReader(
    ActiveRegionScaleFactors=(PositiveWidth=2,NegativeWidth=2,PositiveHeight=2.5,NegativeHeight=0.5)
    ActiveRegionSize=0.1
    Debug=(bDrawDebug=true,DebugScale=25,DebugSegments=20,DebugThickness=0.25)
    DriverAxis=(X=0,Y=1,Z=0)
    DriverItem=(Type=Bone,Name="head")
    FalloffRegionScaleFactors=(PositiveWidth=1,NegativeWidth=1,PositiveHeight=1.5,NegativeHeight=0.7)
    FalloffSize=0.25
    FlipHeightScaling=true
    FlipWidthScaling=false
    OptionalParentItem=(Type=Bone)
    RotationOffset=(X=35,Y=0,Z=0)
   )
   > &RigUnit_ModifyTransforms(
    ItemToModify="((Item=(Type=Bone,Name=\"dyn_collar_side_r_parent\"),Transform=(Rotation=(X=0,Y=0.3007,Z=0,W=0.9537),Translation=(X=-1,Y=0,Z=-1),Scale3D=(X=0.7,Y=1,Z=1))),(Item=(Type=Bone,Name=\"dyn_collar_fr_r_parent\"),Transform=(Rotation=(X=-0.2064,Y=0.2936,Z=0.0651,W=0.93 ...(truncated)"
    WeightMinimum=0
    Weight=@n10.OutputParam
   )
   Spherical Collision = RigVMFunctionReferenceNode(
    Debug=false
    Max Radius=15
    Min Radius=10
    Source=(Type=Bone,Name="dyn_collar_back_parent")
    Source Offset=(X=0,Y=0,Z=0)
    Target=(Type=Bone,Name="head")
    Target Offset=(X=-1,Y=-1,Z=0)
    Weight=1
    ExecuteContext=@RigVMFunction_Sequence.D
   )
   > Spherical Collision_1 = RigVMFunctionReferenceNode(
    Debug=false
    Max Radius=10
    Min Radius=8
    Source=(Type=Bone,Name="dyn_collar_side_l_parent")
    Source Offset=(X=0,Y=0,Z=0)
    Target=(Type=Bone,Name="head")
    Target Offset=(X=-1,Y=0,Z=-2)
    Weight=1
   )
   > Spherical Collision_1_1 = RigVMFunctionReferenceNode(
    Debug=true
    Max Radius=15
    Min Radius=9
    Source=(Type=Bone,Name="dyn_collar_fr_l_parent")
    Source Offset=(X=0,Y=0,Z=0)
    Target=(Type=Bone,Name="head")
    Target Offset=(X=-2,Y=4,Z=-1)
    Weight=1
   )
   > Spherical Collision_1_2 = RigVMFunctionReferenceNode(
    Debug=false
    Max Radius=15
    Min Radius=8
    Source=(Type=Bone,Name="dyn_collar_side_r_parent")
    Source Offset=(X=0,Y=0,Z=0)
    Target=(Type=Bone,Name="head")
    Target Offset=(X=-1,Y=0,Z=1)
    Weight=1
   )
   > Spherical Collision_1_1_1 = RigVMFunctionReferenceNode(
    Debug=false
    Max Radius=15
    Min Radius=7
    Source=(Type=Bone,Name="dyn_collar_fr_r_parent")
    Source Offset=(X=0,Y=0,Z=0)
    Target=(Type=Bone,Name="head")
    Target Offset=(X=-1,Y=4,Z=1)
    Weight=1
   )
   > Spherical Collision_1_3 = RigVMFunctionReferenceNode(
    Debug=false
    Max Radius=8
    Min Radius=5.5
    Source=(Type=Bone,Name="dyn_collar_bk_l_parent")
    Source Offset=(X=0,Y=0,Z=0)
    Target=(Type=Bone,Name="head")
    Target Offset=(X=-1,Y=-2,Z=-1)
    Weight=1
   )
   > Spherical Collision_1_2_1 = RigVMFunctionReferenceNode(
    Debug=false
    Max Radius=15
    Min Radius=8
    Source=(Type=Bone,Name="dyn_collar_bk_r_parent")
    Source Offset=(X=0,Y=0,Z=0)
    Target=(Type=Bone,Name="head")
    Target Offset=(X=-1,Y=-2,Z=0)
    Weight=1
   )
  ]}
]
hierarchy [
 root {loc=<0 0 0>, rot=<0 0 0 1>}
 pelvis {parent=root, loc=<0 -0.648 94.7699>, rot=<0.0127 0.707 -0.0127 -0.707>}
 spine_01 {parent=pelvis, loc=<5.0975 -0.6388 0>, rot=<0 0 0.034 -0.9994>}
 spine_02 {parent=spine_01, loc=<7.2518 -0.035 0>, rot=<0 0 0.0238 -0.9997>}
 spine_03 {parent=spine_02, loc=<6.0104 0 0>, rot=<0 0 0.0149 -0.9999>}
 spine_04 {parent=spine_03, loc=<8.1024 0 0>, rot=<0 0 -0.0132 -0.9999>}
 spine_05 {parent=spine_04, loc=<10.0146 0 0>, rot=<0 0 -0.0015 -1>}
 clavicle_l {parent=spine_05, loc=<4.1651 -0.3023 -5.0942>, rot=<0.0465 -0.7289 -0.009 -0.6829>}
 upperarm_l {parent=clavicle_l, loc=<11.2117 -1.1486 -0.0332>, rot=<0.0179 -0.2982 -0.028 -0.9539>}
 lowerarm_l {parent=upperarm_l, loc=<26.5013 -0.0835 0.2605>, rot=<0.0036 0.0161 -0.2298 -0.9731>}
 hand_l {parent=lowerarm_l, loc=<23.1356 -1.1606 -0.2612>, rot=<-0.6704 -0.0028 0.0567 -0.7398>}
 index_metacarpal_l {parent=hand_l, loc=<1.6535 0.0183 -2.96>, rot=<0.0158 -0.0124 -0.0173 -0.9996>}
 index_01_l {parent=index_metacarpal_l, loc=<8.3008 0.0357 -0.0524>, rot=<-0.0044 0.0485 0.0924 -0.9945>}
 index_02_l {parent=index_01_l, loc=<3.3587 -0.3004 0.0452>, rot=<-0.002 0.0027 0.3225 -0.9466>}
 index_03_l {parent=index_02_l, loc=<2.5578 -0.123 -0.0252>, rot=<-0.002 0.0067 0.2558 -0.9667>}
 middle_metacarpal_l {parent=hand_l, loc=<2.1856 -0.1725 -0.5984>, rot=<0.006 0.0168 -0.0329 -0.9993>}
 middle_01_l {parent=middle_metacarpal_l, loc=<7.3275 -0.2447 0.0129>, rot=<-0.0208 0.0042 0.0858 -0.9961>}
 middle_02_l {parent=middle_01_l, loc=<4.0123 0.1401 0.1175>, rot=<-0.0218 0.0258 0.3512 -0.9357>}
 middle_03_l {parent=middle_02_l, loc=<2.6077 0.0534 0.1373>, rot=<0.0082 0.0188 0.1708 -0.9851>}
 pinky_metacarpal_l {parent=hand_l, loc=<1.7616 -0.9753 3.3814>, rot=<-0.0397 0.0814 -0.069 -0.9935>}
 pinky_01_l {parent=pinky_metacarpal_l, loc=<7.1709 -0.4831 -0.1094>, rot=<-0.0654 -0.0035 0.2242 -0.9723>}
 pinky_02_l {parent=pinky_01_l, loc=<2.8274 0.0893 -0.1043>, rot=<0.0037 0.0011 0.2854 -0.9584>}
 pinky_03_l {parent=pinky_02_l, loc=<2.2046 0.0307 -0.0078>, rot=<-0.0002 0.0098 0.1445 -0.9895>}
 ring_metacarpal_l {parent=hand_l, loc=<2.2074 -0.3625 1.3638>, rot=<-0.0284 0.0722 -0.0458 -0.9959>}
 ring_01_l {parent=ring_metacarpal_l, loc=<7.0295 -0.6197 -0.1446>, rot=<-0.0489 -0.0252 0.1892 -0.9804>}
 ring_02_l {parent=ring_01_l, loc=<3.6174 0.4647 -0.1027>, rot=<0 0 0.2806 -0.9598>}
 ring_03_l {parent=ring_02_l, loc=<2.4152 0.114 0.0199>, rot=<0 0 0.1189 -0.9929>}
 thumb_01_l {parent=hand_l, loc=<1.3118 -0.3191 -2.459>, rot=<0.6546 -0.4042 0.0664 -0.6354>}
 thumb_02_l {parent=thumb_01_l, loc=<5.9336 -0.0448 0.0401>, rot=<-0.0037 0.0213 0.298 -0.9543>}
 thumb_03_l {parent=thumb_02_l, loc=<2.2223 0.3682 -0.0907>, rot=<0.0118 -0.013 0.0066 -0.9998>}
 deform_wrist_dn_base_l {parent=hand_l, loc=<-0.7813 -1.7011 0.9082>, rot=<0 0 0 -1>}
 deform_wrist_dn_l {parent=deform_wrist_dn_base_l, loc=<0 0 0>, rot=<0 0 0 1>}
 deform_wrist_up_base_l {parent=hand_l, loc=<-0.1406 2.5627 0.0698>, rot=<0 0 0 -1>}
 deform_wrist_up_l {parent=deform_wrist_up_base_l, loc=<0 0 0>, rot=<0 0 0 1>}
 lowerarm_twist_01_l {parent=lowerarm_l, loc=<16.7573 0 0>, rot=<0 0 0 1>}
 dyn_stuffing_13_base_l {parent=lowerarm_twist_01_l, loc=<-0.8085 -1.4158 8.9674>, rot=<0.7757 -0.048 0.6152 0.1325>}
 dyn_stuffing_13_l {parent=dyn_stuffing_13_base_l, loc=<0 0 0>, rot=<0 0 0 1>}
 lowerarm_twist_02_l {parent=lowerarm_l, loc=<11.5353 0 0>, rot=<0 0 0 1>}
 dyn_glove_shred_base_l {parent=lowerarm_twist_02_l, loc=<2.1262 8.369 -0.0404>, rot=<-0.2882 -0.4217 0.8543 -0.0968>}
 dyn_glove_shred_l {parent=dyn_glove_shred_base_l, loc=<0 0 0>, rot=<0 0 0 1>}
 dyn_stuffing_12_base_l {parent=lowerarm_twist_02_l, loc=<-0.3988 4.5999 1.9661>, rot=<-0.4061 0.5409 -0.4154 -0.6083>}
 dyn_stuffing_12_l {parent=dyn_stuffing_12_base_l, loc=<0 0 0>, rot=<0 0 0 1>}
 dyn_stuffing_09_base_l {parent=lowerarm_l, loc=<0.6005 1.9395 4.5554>, rot=<-0.3633 0.601 -0.4752 -0.5301>}
 dyn_stuffing_09_l {parent=dyn_stuffing_09_base_l, loc=<0 0 0>, rot=<0 0 0 1>}
 dyn_stuffing_10_base_l {parent=lowerarm_l, loc=<2.6742 -4.7335 9.4993>, rot=<0.81 0.4798 0.2288 -0.2475>}
 dyn_stuffing_10_l {parent=dyn_stuffing_10_base_l, loc=<0 0 0>, rot=<0 0 0 1>}
 dyn_stuffing_06_base_l {parent=lowerarm_l, loc=<8.618 0.5089 4.421>, rot=<0.9889 -0.0403 -0.0581 0.131>}
 dyn_stuffing_06_l {parent=dyn_stuffing_06_base_l, loc=<0 0 0>, rot=<0 0 0 1>}
 deform_elbow_in_base_l {parent=lowerarm_l, loc=<0 4.5494 0>, rot=<0 0 0 -1>}
 deform_elbow_in_l {parent=deform_elbow_in_base_l, loc=<0 0 0>, rot=<0 0 0 1>}
 deform_elbow_out_base_l {parent=lowerarm_l, loc=<-2.6147 -4.6123 0>, rot=<0 0 0 -1>}
 deform_elbow_out_l {parent=deform_elbow_out_base_l, loc=<0 0 0>, rot=<0 0 0 1>}
 upperarm_twist_01_l {parent=upperarm_l, loc=<7.2135 -0.1134 0>, rot=<0 0 0 1>}
 dyn_stuffing_01_base_l {parent=upperarm_twist_01_l, loc=<2.4011 0.8169 1.293>, rot=<0.9997 -0.0019 0.0226 -0.0096>}
 dyn_stuffing_01_l {parent=dyn_stuffing_01_base_l, loc=<0 0 0>, rot=<0 0 0 1>}
 dyn_stuffing_04_base_l {parent=upperarm_twist_01_l, loc=<1.3092 7.3316 3.7648>, rot=<0.6758 -0.472 0.4524 0.3402>}
 dyn_stuffing_04_l {parent=dyn_stuffing_04_base_l, loc=<0 0 0>, rot=<0 0 0 1>}
 dyn_stuffing_05_base_l {parent=upperarm_twist_01_l, loc=<2.8742 -4.0325 2.4597>, rot=<0.6631 0.4102 0.5095 -0.364>}
 dyn_stuffing_05_l {parent=dyn_stuffing_05_base_l, loc=<0 0 0>, rot=<0 0 0 1>}
 deform_upperarm_bk_base_l {parent=upperarm_twist_01_l, loc=<-5.5712 -7.2751 -1.2275>, rot=<0 0 0 -1>}
 deform_upperarm_bk_l {parent=deform_upperarm_bk_base_l, loc=<0 0 0>, rot=<0 0 0 1>}
 deform_upperarm_fr_base_l {parent=upperarm_twist_01_l, loc=<-5.3904 6.218 -0.859>, rot=<0 0 0 -1>}
 deform_upperarm_fr_l {parent=deform_upperarm_fr_base_l, loc=<0 0 0>, rot=<0 0 0 1>}
 deform_upperarm_up_base_l {parent=upperarm_twist_01_l, loc=<-8.9945 -0.5206 4.5756>, rot=<0 0 0 -1>}
 deform_upperarm_up_l {parent=deform_upperarm_up_base_l, loc=<0 0 0>, rot=<0 0 0 1>}
 upperarm_twist_02_l {parent=upperarm_l, loc=<19.7998 -0.1134 0>, rot=<0 0 0 1>}
 dyn_stuffing_08_base_l {parent=upperarm_twist_02_l, loc=<0.3542 1.2471 11.2425>, rot=<0.3086 -0.6438 0.6304 0.3048>}
 dyn_stuffing_08_l {parent=dyn_stuffing_08_base_l, loc=<0 0 0>, rot=<0 0 0 1>}
 deform_bicep_base_l {parent=upperarm_twist_02_l, loc=<-1.8082 4.197 0>, rot=<0 0 0 -1>}
 deform_bicep_l {parent=deform_bicep_base_l, loc=<0 0 0>, rot=<0 0 0 1>}
 dyn_bunny_head_base {parent=upperarm_l, loc=<2.4168 -1.3318 8.256>, rot=<0.3048 0.6304 0.6438 -0.3086>}
 dyn_bunny_head {parent=dyn_bunny_head_base, loc=<0 0 0>, rot=<0 0 0 1>}
 dyn_bunny_head_ear_l {parent=dyn_bunny_head, loc=<6.044 1.8748 2.801>, rot=<-0.2026 0 -0.225 -0.9531>}
 dyn_bunny_head_ear_01_l {parent=dyn_bunny_head_ear_l, loc=<7.0583 -1.4959 -2.7981>, rot=<-0.4775 -0.0336 0.842 0.2489>}
 dyn_bunny_head_ear_r {parent=dyn_bunny_head, loc=<5.9357 -3.8008 -0.5302>, rot=<-0.2549 -0.0151 0.173 -0.9513>}
 dyn_bunny_head_ear_01_r {parent=dyn_bunny_head_ear_r, loc=<9.7771 0.6583 -1.8638>, rot=<0.8932 0.0259 -0.448 0.0292>}
 dyn_stuffing_02_base_l {parent=upperarm_l, loc=<-2.8707 5.709 1.1199>, rot=<0.334 -0.5864 0.6841 0.2767>}
 dyn_stuffing_02_l {parent=dyn_stuffing_02_base_l, loc=<0 0 0>, rot=<0 0 0 1>}
 clavicle_r {parent=spine_05, loc=<4.1652 -0.3023 5.0942>, rot=<0.7289 0.0465 -0.6829 0.009>}
 upperarm_r {parent=clavicle_r, loc=<-11.2117 1.1486 0.0332>, rot=<0.0179 -0.2982 -0.028 -0.9539>}
 lowerarm_r {parent=upperarm_r, loc=<-26.5014 0.0835 -0.2605>, rot=<0.0036 0.0161 -0.2298 -0.9731>}
 hand_r {parent=lowerarm_r, loc=<-23.1357 1.1606 0.2612>, rot=<-0.6704 -0.0028 0.0567 -0.7398>}
 index_metacarpal_r {parent=hand_r, loc=<-1.6534 -0.0181 2.9601>, rot=<0.0158 -0.0124 -0.0173 -0.9996>}
 index_01_r {parent=index_metacarpal_r, loc=<-8.3008 -0.0357 0.0524>, rot=<-0.0044 0.0485 0.0924 -0.9945>}
 index_02_r {parent=index_01_r, loc=<-3.3589 0.3003 -0.0453>, rot=<-0.002 0.0027 0.3225 -0.9466>}
 index_03_r {parent=index_02_r, loc=<-2.5578 0.123 0.0252>, rot=<-0.002 0.0067 0.2558 -0.9667>}
 middle_metacarpal_r {parent=hand_r, loc=<-2.1856 0.1726 0.5984>, rot=<0.006 0.0168 -0.0329 -0.9993>}
 middle_01_r {parent=middle_metacarpal_r, loc=<-7.3273 0.2447 -0.0129>, rot=<-0.0208 0.0042 0.0858 -0.9961>}
 middle_02_r {parent=middle_01_r, loc=<-4.0124 -0.14 -0.1175>, rot=<-0.0218 0.0258 0.3512 -0.9357>}
 middle_03_r {parent=middle_02_r, loc=<-2.6077 -0.0533 -0.1373>, rot=<0.0082 0.0188 0.1708 -0.9851>}
 pinky_metacarpal_r {parent=hand_r, loc=<-1.7617 0.9753 -3.3814>, rot=<-0.0397 0.0814 -0.069 -0.9935>}
 pinky_01_r {parent=pinky_metacarpal_r, loc=<-7.1707 0.4832 0.1094>, rot=<-0.0654 -0.0035 0.2242 -0.9723>}
 pinky_02_r {parent=pinky_01_r, loc=<-2.8274 -0.0892 0.1043>, rot=<0.0037 0.0011 0.2854 -0.9584>}
 pinky_03_r {parent=pinky_02_r, loc=<-2.2045 -0.031 0.0078>, rot=<-0.0002 0.0098 0.1445 -0.9895>}
 ring_metacarpal_r {parent=hand_r, loc=<-2.2073 0.3627 -1.3638>, rot=<-0.0284 0.0722 -0.0458 -0.9959>}
 ring_01_r {parent=ring_metacarpal_r, loc=<-7.0297 0.6197 0.1446>, rot=<-0.0489 -0.0252 0.1892 -0.9804>}
 ring_02_r {parent=ring_01_r, loc=<-3.6171 -0.4648 0.1027>, rot=<0 0 0.2806 -0.9598>}
 ring_03_r {parent=ring_02_r, loc=<-2.4153 -0.1141 -0.0199>, rot=<0 0 0.1189 -0.9929>}
 thumb_01_r {parent=hand_r, loc=<-1.3119 0.3192 2.459>, rot=<0.6546 -0.4042 0.0664 -0.6354>}
 thumb_02_r {parent=thumb_01_r, loc=<-5.9335 0.0447 -0.0401>, rot=<-0.0037 0.0213 0.298 -0.9543>}
 thumb_03_r {parent=thumb_02_r, loc=<-2.2223 -0.3682 0.0907>, rot=<0.0118 -0.013 0.0066 -0.9998>}
 deform_wrist_dn_base_r {parent=hand_r, loc=<0.7811 1.7016 -0.9083>, rot=<0 0 0 -1>}
 deform_wrist_dn_r {parent=deform_wrist_dn_base_r, loc=<0.0003 -0.0003 0.0001>, rot=<0 0 0 -1>}
 deform_wrist_up_base_r {parent=hand_r, loc=<0.1406 -2.5625 -0.0698>, rot=<0 0 0 -1>}
 deform_wrist_up_r {parent=deform_wrist_up_base_r, loc=<0.0001 -0.0001 0>, rot=<0 0 0 -1>}
 lowerarm_twist_01_r {parent=lowerarm_r, loc=<-16.7573 0 0>, rot=<0 0 0 1>}
 lowerarm_twist_02_r {parent=lowerarm_r, loc=<-11.5354 0 0>, rot=<0 0 0 1>}
 dyn_glove_shred_base_r {parent=lowerarm_twist_02_r, loc=<4.5323 6.5152 -2.9052>, rot=<-0.7172 0.0982 0.0926 -0.6837>}
 dyn_glove_shred_r {parent=dyn_glove_shred_base_r, loc=<0 0 0>, rot=<0 0 0 1>}
 dyn_stuffing_09_base_r {parent=lowerarm_r, loc=<-6.0071 -2.9312 -2.9527>, rot=<0.0172 0.1423 0.9895 0.0202>}
 dyn_stuffing_09_r {parent=dyn_stuffing_09_base_r, loc=<0 0 0>, rot=<0 0 0 1>}
 dyn_stuffing_10_base_r {parent=lowerarm_r, loc=<-11.1211 -9.9716 -6.4797>, rot=<0.2572 0.4916 0.6695 -0.4939>}
 dyn_stuffing_10_r {parent=dyn_stuffing_10_base_r, loc=<0 0 0>, rot=<0 0 0 1>}
 deform_elbow_in_base_r {parent=lowerarm_r, loc=<0 -4.5494 0>, rot=<0 0 0 -1>}
 deform_elbow_in_r {parent=deform_elbow_in_base_r, loc=<0 0 0.0001>, rot=<0 0 0 1>}
 deform_elbow_out_base_r {parent=lowerarm_r, loc=<2.6148 4.6123 0.0001>, rot=<0 0 0 -1>}
 deform_elbow_out_r {parent=deform_elbow_out_base_r, loc=<-0.0001 0 0>, rot=<0 0 0 1>}
 upperarm_twist_01_r {parent=upperarm_r, loc=<-7.2135 0.1134 0>, rot=<0 0 0 1>}
 dyn_stuffing_04_base_r {parent=upperarm_twist_01_r, loc=<-5.1501 -1.9042 -9.6119>, rot=<0.147 0.1067 0.692 -0.6986>}
 dyn_stuffing_04_r {parent=dyn_stuffing_04_base_r, loc=<0 0 0>, rot=<0 0 0 1>}
 dyn_stuffing_05_base_r {parent=upperarm_twist_01_r, loc=<-9.1701 0.5812 -3.3235>, rot=<0.0123 -0.0097 0.9999 0.0016>}
 dyn_stuffing_05_r {parent=dyn_stuffing_05_base_r, loc=<0 0 0>, rot=<0 0 0 1>}
 dyn_stuffing_06_base_r {parent=upperarm_twist_01_r, loc=<-8.0795 -5.3932 -9.6412>, rot=<0.5213 -0.4903 -0.504 -0.4837>}
 dyn_stuffing_06_r {parent=dyn_stuffing_06_base_r, loc=<0 0 0>, rot=<0 0 0 1>}
 dyn_stuffing_08_base_r {parent=upperarm_twist_01_r, loc=<-16.6994 -2.8503 -1.9035>, rot=<-0.032 0.0842 0.9883 -0.1233>}
 dyn_stuffing_08_r {parent=dyn_stuffing_08_base_r, loc=<0 0 0>, rot=<0 0 0 1>}
 dyn_stuffing_07_base_r {parent=upperarm_twist_01_r, loc=<-16.7878 1.2173 -2.8004>, rot=<0.0559 -0.0098 0.9984 0.0012>}
 dyn_stuffing_07_r {parent=dyn_stuffing_07_base_r, loc=<0 0 0>, rot=<0 0 0 1>}
 deform_upperarm_bk_base_r {parent=upperarm_twist_01_r, loc=<5.5713 7.2751 1.2274>, rot=<0 0 0 -1>}
 deform_upperarm_bk_r {parent=deform_upperarm_bk_base_r, loc=<-0.0001 0 0.0002>, rot=<0 0 0 -1>}
 deform_upperarm_fr_base_r {parent=upperarm_twist_01_r, loc=<5.3902 -6.218 0.8592>, rot=<0 0 0 -1>}
 deform_upperarm_fr_r {parent=deform_upperarm_fr_base_r, loc=<0.0001 0 -0.0001>, rot=<0 0 0 -1>}
 deform_upperarm_up_base_r {parent=upperarm_twist_01_r, loc=<8.9943 0.5206 -4.5753>, rot=<0 0 0 -1>}
 deform_upperarm_up_r {parent=deform_upperarm_up_base_r, loc=<0.0002 0 -0.0002>, rot=<0 0 0 -1>}
 dyn_stuffing_03_base_r {parent=upperarm_r, loc=<1.0033 -0.8409 -8.0169>, rot=<0.6192 -0.4274 -0.4692 -0.4624>}
 dyn_stuffing_03_r {parent=dyn_stuffing_03_base_r, loc=<0 0 0>, rot=<0 0 0 1>}
 dyn_stuffing_01_base_r {parent=upperarm_r, loc=<-4.5906 0.6513 -3.665>, rot=<-0.0402 0.011 0.9988 -0.026>}
 dyn_stuffing_01_r {parent=dyn_stuffing_01_base_r, loc=<0 0 0>, rot=<0 0 0 1>}
 upperarm_twist_02_r {parent=upperarm_r, loc=<-19.7998 0.1134 0>, rot=<0 0 0 -1>}
 deform_bicep_base_r {parent=upperarm_twist_02_r, loc=<1.8082 -4.197 0>, rot=<0 0 0 1>}
 deform_bicep_r {parent=deform_bicep_base_r, loc=<0 0 0>, rot=<0 0 0 1>}
 neck_01 {parent=spine_05, loc=<11.3775 -0.6396 0>, rot=<0 0 -0.2003 -0.9797>}
 neck_02 {parent=neck_01, loc=<4.5627 0 0>, rot=<0 0 0.0504 -0.9987>}
 head {parent=neck_02, loc=<5.3914 0 0>, rot=<0 0 0.115 -0.9934>}
 shoulder_master_l {parent=spine_05, loc=<5.5855 -1.6793 -12.1346>, rot=<0.0259 -0.9861 0.0454 -0.1578>}
 shoulder_l {parent=shoulder_master_l, loc=<0 0 0>, rot=<0 0 0 1>}
 elbow_l {parent=shoulder_l, loc=<20.853 -0.3894 -0.0782>, rot=<0.0459 -0.103 -0.5395 -0.8344>}
 wrist_l {parent=elbow_l, loc=<12.7903 -1.1633 -5.6925>, rot=<0.2288 -0.3572 -0.1024 -0.8998>}
 shoulder_master_r {parent=spine_05, loc=<5.5852 -1.6794 12.1346>, rot=<0.9861 0.0259 -0.1578 -0.0454>}
 shoulder_r {parent=shoulder_master_r, loc=<0 0 0>, rot=<0 0 0 -1>}
 elbow_r {parent=shoulder_r, loc=<-20.8528 0.3894 0.0781>, rot=<0.0459 -0.103 -0.5395 -0.8344>}
 wrist_r {parent=elbow_r, loc=<-12.7904 1.1635 5.6924>, rot=<0.2288 -0.3572 -0.1024 -0.8998>}
 dyn_zipper_base {parent=spine_05, loc=<3.6578 10.675 0.0789>, rot=<-0.1616 -0.6485 0.7312 0.1367>}
 dyn_zipper {parent=dyn_zipper_base, loc=<0 0 0>, rot=<0 0 0 1>}
 dyn_scarf_base {parent=spine_05, loc=<9.5057 10.9702 -5.3587>, rot=<0.0438 -0.6084 0.7852 0.1064>}
 dyn_scarf {parent=dyn_scarf_base, loc=<0 0 0>, rot=<0 0 0 1>}
 dyn_scarf_01 {parent=dyn_scarf, loc=<5.052 0.1671 0.2821>, rot=<-0.053 -0.0057 -0.1105 -0.9924>}
 dyn_scarf_02 {parent=dyn_scarf_01, loc=<4.6434 -0.2803 0.0925>, rot=<0.01 -0.0368 -0.0092 -0.9992>}
 dyn_hoodie_string_base_r {parent=spine_05, loc=<5.8124 9.3453 2.8866>, rot=<-0.1313 -0.6948 0.6948 0.1313>}
 dyn_hoodie_string_r {parent=dyn_hoodie_string_base_r, loc=<0 0 0>, rot=<0 0 0 1>}
 dyn_hoodie_string_01_r {parent=dyn_hoodie_string_r, loc=<4.5263 0.449 0.1487>, rot=<0 0.0349 0 -0.9994>}
 dyn_hoodie_string_02_r {parent=dyn_hoodie_string_01_r, loc=<4.5189 0.1995 0.3365>, rot=<0.0373 0.1416 -0.0092 -0.9892>}
 dyn_hoodie_string_base_l {parent=spine_05, loc=<5.76 9.3074 -3.3105>, rot=<-0.1075 -0.5189 0.8304 0.172>}
 dyn_hoodie_string_l {parent=dyn_hoodie_string_base_l, loc=<0 0 0>, rot=<0 0 0 1>}
 dyn_hoodie_string_01_l {parent=dyn_hoodie_string_l, loc=<4.3652 -0.093 0.1541>, rot=<0.1767 0.0557 0.0427 -0.9818>}
 dyn_hoodie_string_02_l {parent=dyn_hoodie_string_01_l, loc=<6.5224 0.4919 0.9573>, rot=<-0.0637 0.1386 -0.088 -0.9844>}
 dyn_hoodie_string_03_l {parent=dyn_hoodie_string_02_l, loc=<4.6829 0.2181 0.1414>, rot=<0.0709 0.0069 -0.0225 -0.9972>}
 dyn_hood_ear_base_r {parent=spine_05, loc=<12.423 -12.1936 5.5288>, rot=<0.1824 -0.7473 -0.6026 0.2123>}
 dyn_hood_ear_r {parent=dyn_hood_ear_base_r, loc=<0 0 0>, rot=<0 0 0 1>}
 dyn_hood_ear_01_r {parent=dyn_hood_ear_r, loc=<4.8088 -0.2609 0.3279>, rot=<-0.0382 0.1057 0.0104 -0.9936>}
 dyn_hood_ear_02_r {parent=dyn_hood_ear_01_r, loc=<5.8308 -1.3814 0.8192>, rot=<-0.0589 0.1008 0.1577 -0.9806>}
 dyn_hood_ear_base_l {parent=spine_05, loc=<11.8564 -11.5795 -4.7581>, rot=<-0.2287 0.5793 0.7594 -0.1882>}
 dyn_hood_ear_l {parent=dyn_hood_ear_base_l, loc=<0 0 0>, rot=<0 0 0 1>}
 dyn_hood_ear_01_l {parent=dyn_hood_ear_l, loc=<5.3316 0.2828 -0.0145>, rot=<0.0031 0.1041 -0.078 -0.9915>}
 dyn_hood_ear_02_l {parent=dyn_hood_ear_01_l, loc=<5.2645 0.2861 0.4285>, rot=<0.0213 0.0751 -0.0983 -0.9921>}
 dyn_collar_back_parent {parent=spine_05, loc=<16.7638 -6.9007 -0.1146>, rot=<0.7065 0.0284 -0.0284 -0.7065>}
 dyn_collar_side_r_base {parent=spine_05, loc=<16.0647 -0.7106 7.3819>, rot=<0.9869 0.0396 -0.0063 -0.1563>}
 dyn_collar_side_r_parent {parent=dyn_collar_side_r_base, loc=<0 0 0>, rot=<0 0 0 1>}
 dyn_collar_side_l_base {parent=spine_05, loc=<16 -0.62 -7.382>, rot=<-0.1563 -0.0063 -0.0396 -0.9869>}
 dyn_collar_side_l_parent {parent=dyn_collar_side_l_base, loc=<0 0 0>, rot=<0 0 0 1>}
 dyn_collar_fr_l_base {parent=spine_05, loc=<14.1657 7.9276 -5.786>, rot=<-0.3682 -0.0148 -0.0373 -0.9289>}
 dyn_collar_fr_l_parent {parent=dyn_collar_fr_l_base, loc=<0 0 0>, rot=<0 0 0 1>}
 dyn_collar_fr_r_base {parent=spine_05, loc=<13.8379 7.2271 4.6284>, rot=<0.9375 0.0376 0.0139 0.3456>}
 dyn_collar_fr_r_parent {parent=dyn_collar_fr_r_base, loc=<0 0 0>, rot=<0 0 0 1>}
 dyn_collar_bk_l_base {parent=spine_05, loc=<16.341 -4.8619 -5.1913>, rot=<0.4094 0.0164 -0.0366 -0.9115>}
 dyn_collar_bk_l_parent {parent=dyn_collar_bk_l_base, loc=<0 0 0>, rot=<0 0 0 1>}
 dyn_collar_bk_r_base {parent=spine_05, loc=<16.3924 -4.787 4.6176>, rot=<0.9306 0.0373 -0.0146 -0.364>}
 dyn_collar_bk_r_parent {parent=dyn_collar_bk_r_base, loc=<0 0 0>, rot=<0 0 0 1>}
 deform_pec_aim_base_l {parent=spine_04, loc=<8.0086 12.8872 -2.9192>, rot=<0.0422 -0.4258 0.4306 -0.7947>}
 deform_pec_aim_l {parent=deform_pec_aim_base_l, loc=<0 0 0>, rot=<0 0 0 1>}
 deform_pec_base_l {parent=deform_pec_aim_l, loc=<16.3279 0 0>, rot=<0 0 0 1>}
 deform_pec_l {parent=deform_pec_base_l, loc=<0 0 0>, rot=<0 0 0 1>}
 deform_pec_aim_base_r {parent=spine_04, loc=<8.0086 12.8872 2.9192>, rot=<-0.4258 -0.0422 0.7947 0.4306>}
 deform_pec_aim_r {parent=deform_pec_aim_base_r, loc=<0 0 0>, rot=<0 0 0 -1>}
 deform_pec_base_r {parent=deform_pec_aim_r, loc=<-16.3279 0.0001 0.0001>, rot=<0 0 0 -1>}
 deform_pec_r {parent=deform_pec_base_r, loc=<0 0 0>, rot=<0 0 0 1>}
 elbow_l_ctrl_test {controlType=EULER_TRANSFORM, parent=spine_03, type=Control, loc=<0 0 0>, rot=<0 0 0 1>}
 elbow_r_ctrl_test {controlType=EULER_TRANSFORM, parent=spine_03, type=Control, loc=<0 0 0>, rot=<0 0 0 1>}
 deform_lat_l_base {parent=spine_03, loc=<5.6046 -4.8405 -13.6119>, rot=<0 0 0 1>}
 deform_lat_l {parent=deform_lat_l_base, loc=<0 0 0>, rot=<0 0 0 1>}
 deform_lat_r_base {parent=spine_03, loc=<5.6045 -4.8405 13.6119>, rot=<0 0 1 0>}
 deform_lat_r {parent=deform_lat_r_base, loc=<-0.0001 0 0>, rot=<0 0 0 -1>}
 front_pocket_base {parent=spine_02, loc=<-2.7596 13.9861 -0.6487>, rot=<0.0282 -0.7065 0.7065 -0.0282>}
 front_pocket_parent {parent=front_pocket_base, loc=<0 0 0>, rot=<0 0 0 1>}
 thigh_l {parent=pelvis, loc=<-2.5624 -0.0642 -10.3834>, rot=<0.0004 0.0055 0.0155 -0.9999>}
 calf_l {parent=thigh_l, loc=<-39.2091 -0.1238 -0.6795>, rot=<0.0006 -0.0161 -0.0391 -0.9991>}
 calf_twist_02_l {parent=calf_l, loc=<-17.5927 -1.7943 -0.2638>, rot=<0.0237 0.0095 -0.0035 -0.9997>}
 foot_l {parent=calf_l, loc=<-43.624 -2.818 -0.713>, rot=<0.044 0.0061 0.0367 -0.9983>}
 ball_l {parent=foot_l, loc=<-4.6464 10.3094 -0.3137>, rot=<0.0001 -0.0001 -0.7063 -0.7079>}
 deform_knee_in_base_l {parent=calf_l, loc=<2.8091 -9.2971 0>, rot=<0 0 0 1>}
 deform_knee_in_l {parent=deform_knee_in_base_l, loc=<0 0 0>, rot=<0 0 0 1>}
 deform_knee_out_base_l {parent=calf_l, loc=<-0.2964 5.0241 0>, rot=<0 0 0 1>}
 deform_knee_out_l {parent=deform_knee_out_base_l, loc=<0 0 0>, rot=<0 0 0 1>}
 thigh_twist_01_l {parent=thigh_l, loc=<-23.0764 0 0>, rot=<0.0013 0 0 -1>}
 sub_thigh_l {parent=thigh_l, loc=<1.8201 0 1.4474>, rot=<0 0.03 0 -0.9996>}
 dyn_frill_side_l {parent=sub_thigh_l, loc=<-7.8894 -0.8742 -8.9258>, rot=<0 0 0 1>}
 dyn_frill_bck_l {parent=sub_thigh_l, loc=<0.3181 -10.9455 0.0049>, rot=<0 0 0 1>}
 driver_frill_frt_l {parent=sub_thigh_l, loc=<-8.014 9.641 -0.0046>, rot=<0 0 0 1>}
 dyn_frill_frt_l {parent=driver_frill_frt_l, loc=<0 0 0>, rot=<0 0 0 1>}
 deform_groin_base_l {parent=thigh_l, loc=<0 8.4477 0>, rot=<0 0 0 -1>}
 deform_groin_l {parent=deform_groin_base_l, loc=<0 0 0>, rot=<0 0 0 1>}
 thigh_r {parent=pelvis, loc=<-2.5623 -0.0642 10.3834>, rot=<0.0055 -0.0004 0.9999 0.0155>}
 calf_r {parent=thigh_r, loc=<39.2093 0.1238 0.6795>, rot=<0.0006 -0.0161 -0.0391 -0.9991>}
 calf_twist_02_r {parent=calf_r, loc=<17.5927 1.7943 0.2638>, rot=<0.0237 0.0095 -0.0035 -0.9997>}
 foot_r {parent=calf_r, loc=<43.6242 2.818 0.713>, rot=<0.044 0.0061 0.0367 -0.9983>}
 ball_r {parent=foot_r, loc=<4.6464 -10.3094 0.3137>, rot=<0.0001 -0.0001 -0.7063 -0.7079>}
 deform_knee_in_base_r {parent=calf_r, loc=<-2.8091 9.2972 0>, rot=<0 0 0 1>}
 deform_knee_in_r {parent=deform_knee_in_base_r, loc=<0 0 0>, rot=<0 0 0 1>}
 deform_knee_out_base_r {parent=calf_r, loc=<0.2963 -5.0241 0>, rot=<0 0 0 1>}
 deform_knee_out_r {parent=deform_knee_out_base_r, loc=<0 0 0>, rot=<0 0 0 1>}
 thigh_twist_01_r {parent=thigh_r, loc=<23.0764 0 0>, rot=<0.0013 0 0 -1>}
 sub_thigh_r {parent=thigh_r, loc=<-1.8201 0 -1.4474>, rot=<0 0.03 0 -0.9996>}
 dyn_frill_side_r {parent=sub_thigh_r, loc=<7.8895 0.8742 8.9259>, rot=<0 0 0 1>}
 dyn_frill_bck_r {parent=sub_thigh_r, loc=<-0.4321 10.2417 0.6535>, rot=<0 0 0 1>}
 driver_frill_frt_r {parent=thigh_r, loc=<6.1784 -9.6391 -0.9624>, rot=<0 0.03 0 -0.9996>}
 dyn_frill_frt_r {parent=driver_frill_frt_r, loc=<0 0 0>, rot=<0 0 0 1>}
 deform_groin_base_r {parent=thigh_r, loc=<0.0001 -8.4477 0>, rot=<0 0 0 -1>}
 deform_groin_r {parent=deform_groin_base_r, loc=<0 0 0>, rot=<0 0 0 1>}
 Flap_Front_Parent_Space {parent=pelvis, loc=<-8.9447 10.185 0>, rot=<-0.0127 -0.707 0.0127 -0.707>}
 Flap_Front_Driven_Bone {parent=Flap_Front_Parent_Space, loc=<0 0 0>, rot=<0 0 0 1>}
 dyn_frill_frt {parent=Flap_Front_Driven_Bone, loc=<0 0 0>, rot=<0 0 0 1>}
 dyn_frill_bck {parent=pelvis, loc=<-1.2848 -11.9709 0.429>, rot=<-0.0127 -0.707 0.0127 -0.707>}
 dyn_keychain_base {parent=pelvis, loc=<3.5265 1.9267 -17.0416>, rot=<-0.2744 -0.2114 0.9374 -0.0358>}
 dyn_keychain {parent=dyn_keychain_base, loc=<0 0 0>, rot=<0 0 0 1>}
 dyn_keychain_01 {parent=dyn_keychain, loc=<3.1573 0.0088 -0.0638>, rot=<0.0697 0.0308 0.0164 -0.997>}
 deform_glute_base_l {parent=pelvis, loc=<-13.1199 -13.5027 -12.4424>, rot=<-0.011 0.0565 -0.4248 -0.9035>}
 deform_glute_l {parent=deform_glute_base_l, loc=<-0.0103 -0.0257 0>, rot=<0 0 0 1>}
 deform_glute_base_r {parent=pelvis, loc=<-13.1198 -13.5027 12.4424>, rot=<0.0565 0.011 0.9035 -0.4248>}
 deform_glute_r {parent=deform_glute_base_r, loc=<0 0 0>, rot=<0 0 0 1>}
 shoulder_l_ctrl {controlType=EULER_TRANSFORM, type=Control, loc=<0 0 0>, rot=<0 0 0 1>}
 elbow_l_ctrl {controlType=EULER_TRANSFORM, parent=shoulder_l_ctrl, type=Control, loc=<0 0 0>, rot=<0 0 0 1>}
 wrist_l_ctrl {controlType=EULER_TRANSFORM, parent=elbow_l_ctrl, type=Control, loc=<0 0 0>, rot=<0 0 0 1>}
]
variables [
 crouch still {type=double, default=0}
 head_SPR_look_down_alpha {type=double, default=0}
 head_SPR_look_left_alpha {type=double, default=0}
 head_SPR_look_right_alpha {type=double, default=0}
]
