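# Physics asset for the ears of the large female purple bunny character body, defining collision shapes and joints for the dynamic ear bones.
# A sphere sits on the head bone, and each ear has two capsule segments constrained in a chain from that head bone.
# Joints limit swing and allow very little twist, with outer segments looser than bases.
# Left and right differ, with tighter swing limits on the right ear.
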PhysicsAsset /BRCosmetics/Characters/Player/Female/Large/Bodies/F_LRG_Bunny_Purple/Meshes/F_LRG_Bunny_Purple_Ears_Physics
bodies {
 dyn_bunny_head [{shapeType=Sphere, radius=2.75, shapeName=dyn_bunny_head_sphere}]
 dyn_bunny_head_ear_01_l [{shapeType=Capsule, length=9.2027, radius=4.0729, shapeName=dyn_bunny_head_ear_01_l_capsule, center=<7.3155 0.0002 0>, rotation=<rot 90 0 0>}]
 dyn_bunny_head_ear_01_r [{shapeType=Capsule, length=6.3865, radius=4.241, shapeName=dyn_bunny_head_ear_01_r_capsule, center=<6.2253 -0.0004 0>, rotation=<rot 90 0 0>}]
 dyn_bunny_head_ear_l [{shapeType=Capsule, length=7.2684, radius=4.1736, shapeName=dyn_bunny_head_ear_l_capsule, center=<6.9232 0 0.0004>, rotation=<rot 90 0 0>}]
 dyn_bunny_head_ear_r [{shapeType=Capsule, length=9.283, radius=4.3454, shapeName=dyn_bunny_head_ear_r_capsule, center=<7.9618 0 0>, rotation=<rot 90 0 0>}]
}
constraints [
 {bone1Name=dyn_bunny_head_ear_l, bone2Name=dyn_bunny_head, swing1LimitDegrees=10, swing2LimitDegrees=25, twistLimitDegrees=3}
 {bone1Name=dyn_bunny_head_ear_01_l, bone2Name=dyn_bunny_head_ear_l, swing1LimitDegrees=25, swing2LimitDegrees=30, twistLimitDegrees=3}
 {bone1Name=dyn_bunny_head_ear_r, bone2Name=dyn_bunny_head, swing1LimitDegrees=3, swing2LimitDegrees=20, twistLimitDegrees=2}
 {bone1Name=dyn_bunny_head_ear_01_r, bone2Name=dyn_bunny_head_ear_r, swing1LimitDegrees=15, swing2LimitDegrees=25, twistLimitDegrees=3}
]
