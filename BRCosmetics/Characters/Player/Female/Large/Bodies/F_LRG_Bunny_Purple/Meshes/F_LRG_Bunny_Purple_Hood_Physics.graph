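# Physics asset for the hood of the large female purple bunny player cosmetic, providing bodies and joint limits for its two dynamic ears.
# Each ear is a chain of three capsules hung from a small base sphere, with tight twist limits and swing limits that widen toward the tip.
# An unconstrained spine_05 body combines a large sphere and capsule, forming an upper-torso volume apart from the ear chains.
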PhysicsAsset /BRCosmetics/Characters/Player/Female/Large/Bodies/F_LRG_Bunny_Purple/Meshes/F_LRG_Bunny_Purple_Hood_Physics
bodies {
 dyn_hood_ear_01_l [{shapeType=Capsule, length=2.2893, radius=1.1977, shapeName=dyn_hood_ear_01_l_capsule, center=<2.0307 0 0>, rotation=<rot 90 0 0>}]
 dyn_hood_ear_01_r [{shapeType=Capsule, length=2.9243, radius=1.5065, shapeName=dyn_hood_ear_01_r_capsule, center=<2.9066 0 -0.0001>, rotation=<rot 90 0 0>}]
 dyn_hood_ear_02_l [{shapeType=Capsule, length=1.6249, radius=1.7455, shapeName=dyn_hood_ear_02_l_capsule, center=<1.8785 0 0>, rotation=<rot 90 0 0>}]
 dyn_hood_ear_02_r [{shapeType=Capsule, length=2.3865, radius=2.2087, shapeName=dyn_hood_ear_02_r_capsule, center=<2.0291 0 0>, rotation=<rot 90 0 0>}]
 dyn_hood_ear_base_l [{shapeType=Sphere, radius=0.5, shapeName=dyn_hood_ear_base_l_sphere}]
 dyn_hood_ear_base_r [{shapeType=Sphere, radius=0.5, shapeName=dyn_hood_ear_base_r_sphere}]
 dyn_hood_ear_l [{shapeType=Capsule, length=1.9304, radius=2.5202, shapeName=dyn_hood_ear_l_capsule, center=<2.4915 0 0.0001>, rotation=<rot 90 0 0>}]
 dyn_hood_ear_r [{shapeType=Capsule, length=1.5, radius=2.255, shapeName=dyn_hood_ear_r_capsule, center=<2.4008 0 0>, rotation=<rot 90 0 0>}]
 spine_05 [
  {shapeType=Sphere, radius=12.1477, shapeName=spine_05_sphere, center=<4.8984 -3.4226 0>}
  {shapeType=Capsule, length=12, radius=12.5, shapeName=spine_05_capsule, center=<-4.0636 -1.0906 0>, rotation=<rot 90 -90 -81.682>}
 ]
}
constraints [
 {bone1Name=dyn_hood_ear_r, bone2Name=dyn_hood_ear_base_r, swing1LimitDegrees=5, swing2LimitDegrees=10, twistLimitDegrees=2}
 {bone1Name=dyn_hood_ear_01_r, bone2Name=dyn_hood_ear_r, swing1LimitDegrees=20, swing2LimitDegrees=15, twistLimitDegrees=3}
 {bone1Name=dyn_hood_ear_02_r, bone2Name=dyn_hood_ear_01_r, swing1LimitDegrees=25, swing2LimitDegrees=20, twistLimitDegrees=3}
 {bone1Name=dyn_hood_ear_l, bone2Name=dyn_hood_ear_base_l, swing1LimitDegrees=5, swing2LimitDegrees=10, twistLimitDegrees=2}
 {bone1Name=dyn_hood_ear_01_l, bone2Name=dyn_hood_ear_l, swing1LimitDegrees=20, swing2LimitDegrees=15, twistLimitDegrees=3}
 {bone1Name=dyn_hood_ear_02_l, bone2Name=dyn_hood_ear_01_l, swing1LimitDegrees=25, swing2LimitDegrees=20, twistLimitDegrees=3}
]
