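# Physics asset for the Frontend variant of the purple bunny jacket on the large female player body, defining simulation bodies for its scarf and two hoodie drawstrings.
# Each dangling piece is a capsule chain constrained link by link from a small base sphere, with narrow twist limits and mostly tighter swing on the longer left string.
# Unconstrained spine_04 and spine_05 capsules form torso volumes for the chains to collide against.
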PhysicsAsset /BRCosmetics/Characters/Player/Female/Large/Bodies/F_LRG_Bunny_Purple/Meshes/F_LRG_Bunny_Purple_Jacket_Frontend_Physics
bodies {
 dyn_hoodie_string_01_l [{shapeType=Capsule, length=4.505, radius=1.505, shapeName=dyn_hoodie_string_01_l_capsule, center=<3.0478 0 0.0003>, rotation=<rot 90 0 0>}]
 dyn_hoodie_string_01_r [{shapeType=Capsule, length=3.005, radius=1.005, shapeName=dyn_hoodie_string_01_r_capsule, center=<1.8729 0 0>, rotation=<rot -90 0 0>}]
 dyn_hoodie_string_02_l [{shapeType=Capsule, length=3.8425, radius=1.25, shapeName=dyn_hoodie_string_02_l_capsule, center=<2.3797 0 0.0001>, rotation=<rot 90 0 0>}]
 dyn_hoodie_string_02_r [{shapeType=Capsule, length=5.2052, radius=1, shapeName=dyn_hoodie_string_02_r_capsule, center=<3.1047 0 0>, rotation=<rot 90 0 0>}]
 dyn_hoodie_string_03_l [{shapeType=Capsule, length=0.755, radius=1.505, shapeName=dyn_hoodie_string_03_l_capsule, center=<1.3935 0 0>, rotation=<rot 90 0 0>}]
 dyn_hoodie_string_base_l [{shapeType=Sphere, radius=0.5, shapeName=dyn_hoodie_string_base_l_sphere}]
 dyn_hoodie_string_base_r [{shapeType=Sphere, radius=0.5, shapeName=dyn_hoodie_string_base_r_sphere}]
 dyn_hoodie_string_l [{shapeType=Capsule, length=3.4484, radius=1.25, shapeName=dyn_hoodie_string_l_capsule, center=<2.0884 0 -0.0001>, rotation=<rot 90 0 0>}]
 dyn_hoodie_string_r [{shapeType=Capsule, length=2.255, radius=1.005, shapeName=dyn_hoodie_string_r_capsule, center=<1.9538 0.139 0>, rotation=<rot 90 0 0>}]
 dyn_scarf [{shapeType=Capsule, length=2.8806, radius=2.2197, shapeName=dyn_scarf_capsule, center=<2.9983 0.0001 0.4734>, rotation=<rot 0 0 89.9999>}]
 dyn_scarf_01 [{shapeType=Capsule, length=3.0329, radius=3.3881, shapeName=dyn_scarf_01_capsule, center=<3.239 0 -0.1467>, rotation=<rot 0 0 89.9999>}]
 dyn_scarf_02 [{shapeType=Capsule, length=6.1769, radius=1.9114, shapeName=dyn_scarf_02_capsule, center=<6.1134 0.2086 0.2335>, rotation=<rot 0 0 89.9999>}]
 dyn_scarf_base [{shapeType=Sphere, radius=0.5, shapeName=dyn_scarf_base_sphere}]
 spine_04 [
  {shapeType=Capsule, length=2.5, radius=10.75, shapeName=spine_04_capsule, center=<-0.8989 -0.5495 2.9721>}
  {shapeType=Capsule, length=7.75, radius=8.5, shapeName=spine_04_capsule, center=<-0.8402 0.0796 -7.2946>, rotation=<rot 0 0 10.0002>}
 ]
 spine_05 [{shapeType=Capsule, length=27, radius=12.5, shapeName=spine_05_capsule, center=<-4.6931 -1.1155 0.0001>, rotation=<rot 90 0 3.324>}]
}
constraints [
 {bone1Name=dyn_scarf, bone2Name=dyn_scarf_base, swing1LimitDegrees=20, swing2LimitDegrees=10, twistLimitDegrees=3}
 {bone1Name=dyn_scarf_01, bone2Name=dyn_scarf, swing1LimitDegrees=20, swing2LimitDegrees=15, twistLimitDegrees=3}
 {bone1Name=dyn_scarf_02, bone2Name=dyn_scarf_01, swing1LimitDegrees=20, swing2LimitDegrees=20, twistLimitDegrees=3}
 {bone1Name=dyn_hoodie_string_r, bone2Name=dyn_hoodie_string_base_r, swing1LimitDegrees=20, swing2LimitDegrees=5, twistLimitDegrees=3}
 {bone1Name=dyn_hoodie_string_01_r, bone2Name=dyn_hoodie_string_r, swing1LimitDegrees=20, swing2LimitDegrees=10, twistLimitDegrees=3}
 {bone1Name=dyn_hoodie_string_02_r, bone2Name=dyn_hoodie_string_01_r, swing1LimitDegrees=20, swing2LimitDegrees=10, twistLimitDegrees=3}
 {bone1Name=dyn_hoodie_string_l, bone2Name=dyn_hoodie_string_base_l, swing1LimitDegrees=5, swing2LimitDegrees=2, twistLimitDegrees=3}
 {bone1Name=dyn_hoodie_string_01_l, bone2Name=dyn_hoodie_string_l, swing1LimitDegrees=5, swing2LimitDegrees=5, twistLimitDegrees=3}
 {bone1Name=dyn_hoodie_string_02_l, bone2Name=dyn_hoodie_string_01_l, swing1LimitDegrees=5, swing2LimitDegrees=5, twistLimitDegrees=3}
 {bone1Name=dyn_hoodie_string_03_l, bone2Name=dyn_hoodie_string_02_l, swing1LimitDegrees=10, swing2LimitDegrees=10, twistLimitDegrees=5}
]
