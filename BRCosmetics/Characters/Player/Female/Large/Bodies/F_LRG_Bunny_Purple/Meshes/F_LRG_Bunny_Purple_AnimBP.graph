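# Post-process style AnimBlueprint for the F_LRG_Bunny_Purple player cosmetic, adding secondary motion to a pose copied from the parent mesh.
# Variable-driven bone offsets float and spin stuffing, then RigidBody and AnimDynamics simulate jacket, hood, skirt, zipper and ears.
# Crouch and skydive alphas reposition the keychain and bunny head, and a Control Rig follows.
# A cached pose feeds spring-based AnimDynamics on stuffing and glove shreds before output.
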AnimBlueprint /BRCosmetics/Characters/Player/Female/Large/Bodies/F_LRG_Bunny_Purple/Meshes/F_LRG_Bunny_Purple_AnimBP
target_skeleton=&/F_LRG_Bunny_Purple_Skeleton
graphs [
 AnimGraph {nodes [
   "Constraint - wrist_l (PPPPP)"!(
    ComponentPose
    Alpha=1
    ConstraintWeights_0=0.0885
    ConstraintWeights_1:float=0.0104
    ConstraintWeights_2:float=0.1145
    ConstraintWeights_3:float=0.5226
    ConstraintWeights_4:float=0.264
   ) -> (Pose) {
    alphaBoolBlend {blendOption=Linear}
    boneToModify {boneName=wrist_l}
    constraintSetup [
     {targetBone {boneName=spine_02}}
     {targetBone {boneName=thigh_twist_01_l}}
     {targetBone {boneName=spine_01}}
     {targetBone {boneName=pelvis}}
     {targetBone {boneName=spine_03}}
    ]
    constraintWeights [1,1,1,1,1]
   }
   "Transform (Modify) Bone - Bone: dyn_frill_frt"(Alpha=1, Translation="0,0,-15", Rotation="0,0,0", Scale="1,1,1") {
    rotationMode=BMM_Additive
    rotationSpace=BCS_BoneSpace
    translationMode=BMM_Additive
    translationSpace=BCS_BoneSpace
    alphaBoolBlend {blendOption=Linear}
    alphaScaleBiasClamp {interpSpeedDecreasing=4, interpSpeedIncreasing=6}
    boneToModify {boneName=dyn_frill_frt}
    translation {z=-15}
   }
   n1 = Get Head_Rot_Back() -> (Head_Rot_Back:double)
   n2 = "Transform (Modify) Bone - Bone: dyn_collar_back_parent"(Alpha=@n1, Translation="0,0,0", Rotation="-110,0.000,0", Scale="1,1,1") {
    rotationMode=BMM_Additive
    rotationSpace=BCS_BoneSpace
    alphaBoolBlend {blendOption=Linear}
    alphaScaleBiasClamp {bInterpResult=true, interpSpeedIncreasing=200}
    boneToModify {boneName=dyn_collar_back_parent}
    rotation {pitch=-110}
   }
   "Transform (Modify) Bone - Bone: dyn_collar_side_r_parent"(ComponentPose=@n2, Alpha=@n1, Translation="0,0,0", Rotation="-80,0.000,30", Scale="1,1,1") {
    rotationMode=BMM_Additive
    rotationSpace=BCS_BoneSpace
    alphaBoolBlend {blendOption=Linear}
    alphaScaleBiasClamp {bInterpResult=true, interpSpeedIncreasing=200}
    boneToModify {boneName=dyn_collar_side_r_parent}
    rotation {pitch=-80, roll=30}
   }
   RigidBody() {
    bClampLinearTranslationLimitToRefPose=true
    bOverrideWorldGravity=true
    overridePhysicsAsset=&/F_LRG_Bunny_Purple_Ears_Physics
    simulationSpace=BaseBoneSpace
    alphaBoolBlend {blendOption=Linear}
    baseBoneRef {boneName=spine_05}
    componentAppliedLinearAccClamp=<6000 6000 6000>
    componentLinearAccScale=<0.25 0.25 0.5>
    componentLinearVelScale=<0.25 0.25 0.5>
    overrideWorldGravity {z=-980}
   }
   n3 = Anim Dynamics - Bone: dyn_stuffing_01_l(ComponentPose=Local To Component(LocalPose=Use cached pose 'CachePose'()), Alpha=0.2, ExternalForce="0,0,0") {
    angularDampingOverride=0.99
    angularSpringConstant=150
    bAngularSpring=true
    bGravityOverrideInSimSpace=true
    bLinearSpring=true
    bOverrideAngularDamping=true
    bOverrideLinearDamping=true
    gravityScale=0
    linearDampingOverride=0.995
    linearSpringConstant=50
    simulationSpace=BoneRelative
    alphaBoolBlend {blendOption=Linear}
    boundBone {boneName=dyn_stuffing_01_l}
    componentLinearAccScale=<0.05 0.05 0.05>
    componentLinearVelScale=<0.8 0.8 0.6>
    physicsBodyDefinitions [{
      sphereCollisionRadius=0
      boundBone {boneName=dyn_stuffing_01_l}
      boxExtents=<50 50 50>
      constraintSetup {angularLimitsMax=<0 0 0>, angularLimitsMin=<0 0 0>, angularTarget=<0 0 0>, linearAxesMax=<2 3 5>, linearAxesMin=<-2 -1 -5>}
      localJointOffset=<0 -3 0>
     }]
    relativeSpaceBone {boneName=spine_05}
   }
   n4 = Anim Dynamics - Bone: dyn_stuffing_05_l(ComponentPose=@n3, Alpha=0.2, ExternalForce="0,0,0") {
    angularDampingOverride=0.99
    angularSpringConstant=150
    bAngularSpring=true
    bGravityOverrideInSimSpace=true
    bLinearSpring=true
    bOverrideAngularDamping=true
    bOverrideLinearDamping=true
    gravityScale=0
    linearDampingOverride=0.9999
    linearSpringConstant=50
    simulationSpace=BoneRelative
    alphaBoolBlend {blendOption=Linear}
    boundBone {boneName=dyn_stuffing_05_l}
    componentLinearAccScale=<0.05 0.05 0.05>
    componentLinearVelScale=<0.8 1 0.6>
    physicsBodyDefinitions [{
      sphereCollisionRadius=0
      boundBone {boneName=dyn_stuffing_05_l}
      boxExtents=<50 50 50>
      constraintSetup {angularLimitsMax=<0 0 0>, angularLimitsMin=<0 0 0>, angularTarget=<0 0 0>, linearAxesMax=<2 5 5>, linearAxesMin=<-2 1 -5>}
      localJointOffset=<0 -3 0>
     }]
    relativeSpaceBone {boneName=spine_05}
   }
   n5 = Anim Dynamics - Bone: dyn_stuffing_09_l(ComponentPose=@n4, Alpha=0.2, ExternalForce="0,0,0") {
    angularDampingOverride=0.99
    angularSpringConstant=150
    bAngularSpring=true
    bGravityOverrideInSimSpace=true
    bLinearSpring=true
    bOverrideAngularDamping=true
    bOverrideLinearDamping=true
    gravityScale=0
    linearDampingOverride=0.995
    linearSpringConstant=50
    simulationSpace=BoneRelative
    alphaBoolBlend {blendOption=Linear}
    boundBone {boneName=dyn_stuffing_09_l}
    componentLinearAccScale=<0.1 0.1 0.3>
    componentLinearVelScale=<0.8 1 0.6>
    physicsBodyDefinitions [{
      sphereCollisionRadius=0
      boundBone {boneName=dyn_stuffing_09_l}
      boxExtents=<50 50 50>
      constraintSetup {angularLimitsMax=<0 0 0>, angularLimitsMin=<0 0 0>, angularTarget=<0 0 0>, linearAxesMax=<2 5 5>, linearAxesMin=<-2 1 -5>}
      localJointOffset=<0 -3 0>
     }]
    relativeSpaceBone {boneName=spine_05}
   }
   n6 = Anim Dynamics - Bone: dyn_glove_shred_l(ComponentPose=@n5, Alpha=0.2, ExternalForce="0,0,0") {
    angularDampingOverride=0.9
    angularSpringConstant=150
    bAngularSpring=true
    bGravityOverrideInSimSpace=true
    bLinearSpring=true
    bOverrideAngularDamping=true
    bOverrideLinearDamping=true
    gravityScale=0
    linearDampingOverride=0.995
    linearSpringConstant=50
    simulationSpace=BoneRelative
    alphaBoolBlend {blendOption=Linear}
    boundBone {boneName=dyn_glove_shred_l}
    componentLinearAccScale=<0.1 0.1 0.3>
    componentLinearVelScale=<0.8 1 0.6>
    physicsBodyDefinitions [{
      sphereCollisionRadius=0
      boundBone {boneName=dyn_glove_shred_l}
      boxExtents=<50 50 50>
      constraintSetup {angularLimitsMax=<0 0 0>, angularLimitsMin=<0 0 0>, angularTarget=<0 0 0>, linearAxesMax=<2 5 6>, linearAxesMin=<-2 1 -6>}
      localJointOffset=<0 -3 0>
     }]
    relativeSpaceBone {boneName=spine_05}
   }
   n7 = Anim Dynamics - Bone: dyn_stuffing_13_l(ComponentPose=@n6, Alpha=0.2, ExternalForce="0,0,0") {
    angularDampingOverride=0.9
    angularSpringConstant=150
    bAngularSpring=true
    bGravityOverrideInSimSpace=true
    bLinearSpring=true
    bOverrideAngularDamping=true
    bOverrideLinearDamping=true
    gravityScale=0
    linearDampingOverride=0.999
    linearSpringConstant=50
    simulationSpace=BoneRelative
    alphaBoolBlend {blendOption=Linear}
    boundBone {boneName=dyn_stuffing_13_l}
    componentLinearAccScale=<0.1 0.1 0.3>
    componentLinearVelScale=<0.8 1 0.6>
    physicsBodyDefinitions [{
      sphereCollisionRadius=0
      boundBone {boneName=dyn_stuffing_13_l}
      boxExtents=<50 50 50>
      constraintSetup {angularLimitsMax=<0 0 0>, angularLimitsMin=<0 0 0>, angularTarget=<0 0 0>, linearAxesMax=<1 1 1>, linearAxesMin=<-1 0 -1>}
      localJointOffset=<0 -3 0>
     }]
    relativeSpaceBone {boneName=spine_05}
   }
   n8 = Anim Dynamics - Bone: dyn_stuffing_12_l(ComponentPose=@n7, Alpha=0.2, ExternalForce="0,0,0") {
    angularDampingOverride=0.99
    angularSpringConstant=150
    bAngularSpring=true
    bGravityOverrideInSimSpace=true
    bLinearSpring=true
    bOverrideAngularDamping=true
    bOverrideLinearDamping=true
    gravityScale=0
    linearDampingOverride=0.995
    linearSpringConstant=50
    simulationSpace=BoneRelative
    alphaBoolBlend {blendOption=Linear}
    boundBone {boneName=dyn_stuffing_12_l}
    componentLinearAccScale=<0.1 0.1 0.3>
    componentLinearVelScale=<0.8 1 0.6>
    physicsBodyDefinitions [{
      sphereCollisionRadius=0
      boundBone {boneName=dyn_stuffing_12_l}
      boxExtents=<50 50 50>
      constraintSetup {angularLimitsMax=<0 0 0>, angularLimitsMin=<0 0 0>, angularTarget=<0 0 0>, linearAxesMax=<2 5 3>, linearAxesMin=<-2 1 -3>}
      localJointOffset=<0 -3 0>
     }]
    relativeSpaceBone {boneName=spine_05}
   }
   n9 = Anim Dynamics - Bone: dyn_stuffing_10_l(ComponentPose=@n8, Alpha=0.2, ExternalForce="0,0,0") {
    angularDampingOverride=0.9
    angularSpringConstant=150
    bAngularSpring=true
    bGravityOverrideInSimSpace=true
    bLinearSpring=true
    bOverrideAngularDamping=true
    bOverrideLinearDamping=true
    gravityScale=0
    linearDampingOverride=0.999
    linearSpringConstant=50
    simulationSpace=BoneRelative
    alphaBoolBlend {blendOption=Linear}
    boundBone {boneName=dyn_stuffing_10_l}
    componentLinearAccScale=<0.1 0.1 0.3>
    componentLinearVelScale=<0.8 1 0.6>
    physicsBodyDefinitions [{
      sphereCollisionRadius=0
      boundBone {boneName=dyn_stuffing_10_l}
      boxExtents=<50 50 50>
      constraintSetup {angularLimitsMax=<0 0 0>, angularLimitsMin=<0 0 0>, angularTarget=<0 0 0>, linearAxesMax=<2 2 4>, linearAxesMin=<-2 -2 -4>}
      localJointOffset=<0 -3 0>
     }]
    relativeSpaceBone {boneName=spine_05}
   }
   n10 = Anim Dynamics - Bone: dyn_stuffing_08_l(ComponentPose=@n9, Alpha=0.2, ExternalForce="0,0,0") {
    angularDampingOverride=0.9
    angularSpringConstant=150
    bAngularSpring=true
    bGravityOverrideInSimSpace=true
    bLinearSpring=true
    bOverrideAngularDamping=true
    bOverrideLinearDamping=true
    gravityScale=0
    linearDampingOverride=0.999
    linearSpringConstant=50
    simulationSpace=BoneRelative
    alphaBoolBlend {blendOption=Linear}
    boundBone {boneName=dyn_stuffing_08_l}
    componentLinearAccScale=<0.1 0.1 0.3>
    componentLinearVelScale=<0.8 1 0.6>
    physicsBodyDefinitions [{
      sphereCollisionRadius=0
      boundBone {boneName=dyn_stuffing_08_l}
      boxExtents=<50 50 50>
      constraintSetup {angularLimitsMax=<0 0 0>, angularLimitsMin=<0 0 0>, angularTarget=<0 0 0>, linearAxesMax=<2 5 5>, linearAxesMin=<-2 1 -5>}
      localJointOffset=<0 -3 0>
     }]
    relativeSpaceBone {boneName=spine_05}
   }
   n11 = Anim Dynamics - Bone: dyn_stuffing_06_l(ComponentPose=@n10, Alpha=0.2, ExternalForce="0,0,0") {
    angularDampingOverride=0.99
    angularSpringConstant=150
    bAngularSpring=true
    bGravityOverrideInSimSpace=true
    bLinearSpring=true
    bOverrideAngularDamping=true
    bOverrideLinearDamping=true
    gravityScale=0
    linearDampingOverride=0.995
    linearSpringConstant=50
    simulationSpace=BoneRelative
    alphaBoolBlend {blendOption=Linear}
    boundBone {boneName=dyn_stuffing_06_l}
    componentLinearAccScale=<0.1 0.1 0.3>
    componentLinearVelScale=<0.8 1 0.6>
    physicsBodyDefinitions [{
      sphereCollisionRadius=0
      boundBone {boneName=dyn_stuffing_06_l}
      boxExtents=<50 50 50>
      constraintSetup {angularLimitsMax=<0 0 0>, angularLimitsMin=<0 0 0>, angularTarget=<0 0 0>, linearAxesMax=<2 2 2>, linearAxesMin=<-2 1 -2>}
      localJointOffset=<0 -3 0>
     }]
    relativeSpaceBone {boneName=spine_05}
   }
   n12 = Anim Dynamics - Bone: dyn_stuffing_04_l(ComponentPose=@n11, Alpha=0.2, ExternalForce="0,0,0") {
    angularDampingOverride=0.9
    angularSpringConstant=150
    bAngularSpring=true
    bGravityOverrideInSimSpace=true
    bLinearSpring=true
    bOverrideAngularDamping=true
    bOverrideLinearDamping=true
    gravityScale=0
    linearDampingOverride=0.999
    linearSpringConstant=50
    simulationSpace=BoneRelative
    alphaBoolBlend {blendOption=Linear}
    boundBone {boneName=dyn_stuffing_04_l}
    componentLinearAccScale=<0.05 0.05 0.05>
    componentLinearVelScale=<0.8 0.8 0.6>
    physicsBodyDefinitions [{
      sphereCollisionRadius=0
      boundBone {boneName=dyn_stuffing_04_l}
      boxExtents=<50 50 50>
      constraintSetup {angularLimitsMax=<0 0 0>, angularLimitsMin=<0 0 0>, angularTarget=<0 0 0>, linearAxesMax=<2 5 4>, linearAxesMin=<-2 0 -4>}
      localJointOffset=<0 -3 0>
     }]
    relativeSpaceBone {boneName=spine_05}
   }
   n13 = Anim Dynamics - Bone: dyn_stuffing_02_l(ComponentPose=@n12, Alpha=0.2, ExternalForce="0,0,0") {
    angularDampingOverride=0.9
    angularSpringConstant=150
    bAngularSpring=true
    bGravityOverrideInSimSpace=true
    bLinearSpring=true
    bOverrideAngularDamping=true
    bOverrideLinearDamping=true
    gravityScale=0
    linearDampingOverride=0.999
    linearSpringConstant=50
    simulationSpace=BoneRelative
    alphaBoolBlend {blendOption=Linear}
    boundBone {boneName=dyn_stuffing_02_l}
    componentLinearAccScale=<0.1 0.1 0.3>
    componentLinearVelScale=<0.8 0.8 0.6>
    physicsBodyDefinitions [{
      sphereCollisionRadius=4
      boundBone {boneName=dyn_stuffing_02_l}
      boxExtents=<50 50 50>
      constraintSetup {angularLimitsMax=<0 0 0>, angularLimitsMin=<0 0 0>, angularTarget=<0 0 0>, linearAxesMax=<2 5 4>, linearAxesMin=<-2 -1 -4>}
      localJointOffset=<0 -3 0>
     }]
    relativeSpaceBone {boneName=spine_05}
    sphericalLimits [{limitRadius=0, drivingBone {boneName=None}, sphereLocalOffset=<0 0 0>}]
   }
   n14 = Anim Dynamics - Bone: dyn_stuffing_01_r(ComponentPose=@n13, Alpha=0.2, ExternalForce="0,0,0") {
    angularDampingOverride=0.9
    angularSpringConstant=150
    bAngularSpring=true
    bGravityOverrideInSimSpace=true
    bLinearSpring=true
    bOverrideAngularDamping=true
    bOverrideLinearDamping=true
    gravityScale=0
    linearDampingOverride=0.999
    linearSpringConstant=50
    simulationSpace=BoneRelative
    alphaBoolBlend {blendOption=Linear}
    boundBone {boneName=dyn_stuffing_01_r}
    componentLinearAccScale=<0.05 0.05 0.05>
    componentLinearVelScale=<0.8 0.8 0.6>
    physicsBodyDefinitions [{
      sphereCollisionRadius=0
      boundBone {boneName=dyn_stuffing_01_r}
      boxExtents=<50 50 50>
      constraintSetup {angularLimitsMax=<0 0 0>, angularLimitsMin=<0 0 0>, angularTarget=<0 0 0>, linearAxesMax=<2 3 3>, linearAxesMin=<-2 -1 -3>}
      localJointOffset=<0 -3 0>
     }]
    relativeSpaceBone {boneName=spine_05}
   }
   n15 = Anim Dynamics - Bone: dyn_stuffing_03_r(ComponentPose=@n14, Alpha=0.2, ExternalForce="0,0,0") {
    angularDampingOverride=0.9
    angularSpringConstant=150
    bAngularSpring=true
    bGravityOverrideInSimSpace=true
    bLinearSpring=true
    bOverrideAngularDamping=true
    bOverrideLinearDamping=true
    gravityScale=0
    linearDampingOverride=0.999
    linearSpringConstant=50
    simulationSpace=BoneRelative
    alphaBoolBlend {blendOption=Linear}
    boundBone {boneName=dyn_stuffing_03_r}
    componentLinearAccScale=<0.1 0.1 0.3>
    componentLinearVelScale=<0.8 0.8 0.6>
    physicsBodyDefinitions [{
      sphereCollisionRadius=0
      boundBone {boneName=dyn_stuffing_03_r}
      boxExtents=<50 50 50>
      constraintSetup {angularLimitsMax=<0 0 0>, angularLimitsMin=<0 0 0>, angularTarget=<0 0 0>, linearAxesMax=<2 3 5>, linearAxesMin=<-2 0 -5>}
      localJointOffset=<0 -3 0>
     }]
    relativeSpaceBone {boneName=spine_05}
   }
   n16 = Anim Dynamics - Bone: dyn_stuffing_05_r(ComponentPose=@n15, Alpha=0.2, ExternalForce="0,0,0") {
    angularDampingOverride=0.9
    angularSpringConstant=150
    bAngularSpring=true
    bGravityOverrideInSimSpace=true
    bLinearSpring=true
    bOverrideAngularDamping=true
    bOverrideLinearDamping=true
    bUseSphericalLimits=true
    gravityScale=0
    linearDampingOverride=0.999
    linearSpringConstant=50
    simulationSpace=BoneRelative
    alphaBoolBlend {blendOption=Linear}
    boundBone {boneName=dyn_stuffing_05_r}
    componentLinearAccScale=<0.05 0.05 0.05>
    componentLinearVelScale=<0.8 1 0.6>
    physicsBodyDefinitions [{
      sphereCollisionRadius=0
      boundBone {boneName=dyn_stuffing_05_r}
      boxExtents=<50 50 50>
      constraintSetup {angularLimitsMax=<0 0 0>, angularLimitsMin=<0 0 0>, angularTarget=<0 0 0>, linearAxesMax=<2 3 4>, linearAxesMin=<-2 -1 -4>}
      localJointOffset=<0 -3 0>
     }]
    relativeSpaceBone {boneName=spine_05}
   }
   n17 = Anim Dynamics - Bone: dyn_stuffing_07_r(ComponentPose=@n16, Alpha=0.2, ExternalForce="0,0,0") {
    angularDampingOverride=0.99
    angularSpringConstant=150
    bAngularSpring=true
    bGravityOverrideInSimSpace=true
    bLinearSpring=true
    bOverrideAngularDamping=true
    bOverrideLinearDamping=true
    gravityScale=0
    linearDampingOverride=0.995
    linearSpringConstant=50
    simulationSpace=BoneRelative
    alphaBoolBlend {blendOption=Linear}
    boundBone {boneName=dyn_stuffing_07_r}
    componentLinearAccScale=<0.1 0.1 0.3>
    componentLinearVelScale=<0.8 1 0.6>
    physicsBodyDefinitions [{
      sphereCollisionRadius=0
      boundBone {boneName=dyn_stuffing_07_r}
      boxExtents=<50 50 50>
      constraintSetup {angularLimitsMax=<0 0 0>, angularLimitsMin=<0 0 0>, angularTarget=<0 0 0>, linearAxesMax=<2 3 4>, linearAxesMin=<-2 1 -2>}
      localJointOffset=<0 -3 0>
     }]
    relativeSpaceBone {boneName=spine_05}
   }
   n18 = Anim Dynamics - Bone: dyn_stuffing_09_r(ComponentPose=@n17, Alpha=0.2, ExternalForce="0,0,0") {
    angularDampingOverride=0.99
    angularSpringConstant=150
    bAngularSpring=true
    bGravityOverrideInSimSpace=true
    bLinearSpring=true
    bOverrideAngularDamping=true
    bOverrideLinearDamping=true
    gravityScale=0
    linearDampingOverride=0.995
    linearSpringConstant=50
    simulationSpace=BoneRelative
    alphaBoolBlend {blendOption=Linear}
    boundBone {boneName=dyn_stuffing_09_r}
    componentLinearAccScale=<0.1 0.1 0.3>
    componentLinearVelScale=<0.8 1 0.6>
    physicsBodyDefinitions [{
      sphereCollisionRadius=0
      boundBone {boneName=dyn_stuffing_09_r}
      boxExtents=<50 50 50>
      constraintSetup {angularLimitsMax=<0 0 0>, angularLimitsMin=<0 0 0>, angularTarget=<0 0 0>, linearAxesMax=<2 4 5>, linearAxesMin=<-2 1 -5>}
      localJointOffset=<0 -3 0>
     }]
    relativeSpaceBone {boneName=spine_05}
   }
   n19 = Anim Dynamics - Bone: dyn_glove_shred_r(ComponentPose=@n18, Alpha=0.2, ExternalForce="0,0,0") {
    angularDampingOverride=0.9
    angularSpringConstant=150
    bAngularSpring=true
    bGravityOverrideInSimSpace=true
    bLinearSpring=true
    bOverrideAngularDamping=true
    bOverrideLinearDamping=true
    gravityScale=0
    linearDampingOverride=0.995
    linearSpringConstant=50
    simulationSpace=BoneRelative
    alphaBoolBlend {blendOption=Linear}
    boundBone {boneName=dyn_glove_shred_r}
    componentLinearAccScale=<0.1 0.1 0.3>
    componentLinearVelScale=<0.8 1 0.6>
    physicsBodyDefinitions [{
      sphereCollisionRadius=0
      boundBone {boneName=dyn_glove_shred_r}
      boxExtents=<50 50 50>
      constraintSetup {angularLimitsMax=<0 0 0>, angularLimitsMin=<0 0 0>, angularTarget=<0 0 0>, linearAxesMax=<2 3 4>, linearAxesMin=<-2 1 -4>}
      localJointOffset=<0 -3 0>
     }]
    relativeSpaceBone {boneName=spine_05}
   }
   n20 = Anim Dynamics - Bone: dyn_stuffing_10_r(ComponentPose=@n19, Alpha=0.2, ExternalForce="0,0,0") {
    angularDampingOverride=0.9
    angularSpringConstant=150
    bAngularSpring=true
    bGravityOverrideInSimSpace=true
    bLinearSpring=true
    bOverrideAngularDamping=true
    bOverrideLinearDamping=true
    gravityScale=0
    linearDampingOverride=0.999
    linearSpringConstant=50
    simulationSpace=BoneRelative
    alphaBoolBlend {blendOption=Linear}
    boundBone {boneName=dyn_stuffing_10_r}
    componentLinearAccScale=<0.1 0.1 0.3>
    componentLinearVelScale=<0.8 1 0.6>
    physicsBodyDefinitions [{
      sphereCollisionRadius=0
      boundBone {boneName=dyn_stuffing_10_r}
      boxExtents=<50 50 50>
      constraintSetup {angularLimitsMax=<0 0 0>, angularLimitsMin=<0 0 0>, angularTarget=<0 0 0>, linearAxesMax=<2 2 4>, linearAxesMin=<-2 -1 -4>}
      localJointOffset=<0 -3 0>
     }]
    relativeSpaceBone {boneName=spine_05}
   }
   n21 = Anim Dynamics - Bone: dyn_stuffing_08_r(ComponentPose=@n20, Alpha=0.2, ExternalForce="0,0,0") {
    angularDampingOverride=0.99
    angularSpringConstant=150
    bAngularSpring=true
    bGravityOverrideInSimSpace=true
    bLinearSpring=true
    bOverrideAngularDamping=true
    bOverrideLinearDamping=true
    gravityScale=0
    linearDampingOverride=0.995
    linearSpringConstant=50
    simulationSpace=BoneRelative
    alphaBoolBlend {blendOption=Linear}
    boundBone {boneName=dyn_stuffing_08_r}
    componentLinearAccScale=<0.1 0.1 0.3>
    componentLinearVelScale=<0.8 1 0.6>
    physicsBodyDefinitions [{
      sphereCollisionRadius=0
      boundBone {boneName=dyn_stuffing_08_r}
      boxExtents=<50 50 50>
      constraintSetup {angularLimitsMax=<0 0 0>, angularLimitsMin=<0 0 0>, angularTarget=<0 0 0>, linearAxesMax=<2 3 4>, linearAxesMin=<-2 1 -2>}
      localJointOffset=<0 -3 0>
     }]
    relativeSpaceBone {boneName=spine_05}
   }
   n22 = Anim Dynamics - Bone: dyn_stuffing_06_r(ComponentPose=@n21, Alpha=0.2, ExternalForce="0,0,0") {
    angularDampingOverride=0.99
    angularSpringConstant=150
    bAngularSpring=true
    bGravityOverrideInSimSpace=true
    bLinearSpring=true
    bOverrideAngularDamping=true
    bOverrideLinearDamping=true
    gravityScale=0
    linearDampingOverride=0.995
    linearSpringConstant=50
    simulationSpace=BoneRelative
    alphaBoolBlend {blendOption=Linear}
    boundBone {boneName=dyn_stuffing_06_r}
    componentLinearAccScale=<0.1 0.1 0.3>
    componentLinearVelScale=<0.8 1 0.6>
    physicsBodyDefinitions [{
      sphereCollisionRadius=0
      boundBone {boneName=dyn_stuffing_06_r}
      boxExtents=<50 50 50>
      constraintSetup {angularLimitsMax=<0 0 0>, angularLimitsMin=<0 0 0>, angularTarget=<0 0 0>, linearAxesMax=<2 3 4>, linearAxesMin=<-2 1 -4>}
      localJointOffset=<0 -3 0>
     }]
    relativeSpaceBone {boneName=spine_05}
   }
   n23 = Anim Dynamics - Bone: dyn_stuffing_04_r(ComponentPose=@n22, Alpha=0.2, ExternalForce="0,0,0") {
    angularDampingOverride=0.99
    angularSpringConstant=150
    bAngularSpring=true
    bGravityOverrideInSimSpace=true
    bLinearSpring=true
    bOverrideAngularDamping=true
    bOverrideLinearDamping=true
    gravityScale=0
    linearDampingOverride=0.995
    linearSpringConstant=50
    simulationSpace=BoneRelative
    alphaBoolBlend {blendOption=Linear}
    boundBone {boneName=dyn_stuffing_04_r}
    componentLinearAccScale=<0.05 0.05 0.05>
    componentLinearVelScale=<0.8 0.8 0.6>
    physicsBodyDefinitions [{
      sphereCollisionRadius=0
      boundBone {boneName=dyn_stuffing_04_r}
      boxExtents=<50 50 50>
      constraintSetup {angularLimitsMax=<0 0 0>, angularLimitsMin=<0 0 0>, angularTarget=<0 0 0>, linearAxesMax=<2 4 5>, linearAxesMin=<-2 0 -5>}
      localJointOffset=<0 -3 0>
     }]
    relativeSpaceBone {boneName=spine_05}
   }
   Output Pose(Result=Component To Local(ComponentPose=@n23)) {layerGroup=None}
   n24 = Get Stuffing_Float_A() -> (Stuffing_Float_A:Vector)
   n25 = "Transform (Modify) Bone - Bone: dyn_bunny_head"(
    ComponentPose=Local To Component(LocalPose=Copy Pose From Mesh() {bUseAttachedParent=true})
    Alpha=1
    Translation=Get Stuffing_Float_B() -> (Stuffing_Float_B:Vector)
    Rotation="0,0,0"
    Scale="1,1,1"
   ) {
    rotationMode=BMM_Replace
    rotationSpace=BCS_BoneSpace
    translationMode=BMM_Additive
    translationSpace=BCS_BoneSpace
    alphaBoolBlend {blendOption=Linear}
    boneToModify {boneName=dyn_bunny_head}
   }
   n26 = Get Spinning_Rotator() -> (Spinning_Rotator:Rotator)
   n27 = Get Stuffing_Float_A()
   n28 = "Transform (Modify) Bone - Bone: dyn_stuffing_01_r"(ComponentPose=@n25, Alpha=0.5, Translation=@n27, Rotation=@n26, Scale="1,1,1") {
    rotationMode=BMM_Replace
    rotationSpace=BCS_BoneSpace
    translationMode=BMM_Additive
    translationSpace=BCS_BoneSpace
    alphaBoolBlend {blendOption=Linear}
    boneToModify {boneName=dyn_stuffing_01_r}
   }
   n29 = "Transform (Modify) Bone - Bone: dyn_stuffing_03_r"(ComponentPose=@n28, Alpha=0.75, Translation=@n27, Rotation=@n26, Scale="1,1,1") {
    rotationMode=BMM_Replace
    rotationSpace=BCS_BoneSpace
    translationMode=BMM_Additive
    translationSpace=BCS_BoneSpace
    alphaBoolBlend {blendOption=Linear}
    boneToModify {boneName=dyn_stuffing_03_r}
   }
   n30 = "Transform (Modify) Bone - Bone: dyn_stuffing_04_r"(ComponentPose=@n29, Alpha=1, Translation=@n27, Rotation="0,0,0", Scale="1,1,1") {
    rotationMode=BMM_Replace
    rotationSpace=BCS_BoneSpace
    translationMode=BMM_Additive
    translationSpace=BCS_BoneSpace
    alphaBoolBlend {blendOption=Linear}
    boneToModify {boneName=dyn_stuffing_04_r}
   }
   n31 = Get Stuffing_Float_C() -> (Stuffing_Float_C:Vector)
   n32 = "Transform (Modify) Bone - Bone: dyn_stuffing_05_r"(ComponentPose=@n30, Alpha=0.5, Translation=@n31, Rotation="0,0,0", Scale="1,1,1") {
    rotationMode=BMM_Replace
    rotationSpace=BCS_BoneSpace
    translationMode=BMM_Additive
    translationSpace=BCS_BoneSpace
    alphaBoolBlend {blendOption=Linear}
    boneToModify {boneName=dyn_stuffing_05_r}
   }
   n33 = "Transform (Modify) Bone - Bone: dyn_stuffing_07_r"(ComponentPose=@n32, Alpha=0.75, Translation=@n31, Rotation="0,0,0", Scale="1,1,1") {
    rotationMode=BMM_Replace
    rotationSpace=BCS_BoneSpace
    translationMode=BMM_Additive
    translationSpace=BCS_BoneSpace
    alphaBoolBlend {blendOption=Linear}
    boneToModify {boneName=dyn_stuffing_07_r}
   }
   n34 = "Transform (Modify) Bone - Bone: dyn_stuffing_08_r"(ComponentPose=@n33, Alpha=0.5, Translation=@n31, Rotation="0,0,0", Scale="1,1,1") {
    rotationMode=BMM_Replace
    rotationSpace=BCS_BoneSpace
    translationMode=BMM_Additive
    translationSpace=BCS_BoneSpace
    alphaBoolBlend {blendOption=Linear}
    boneToModify {boneName=dyn_stuffing_08_r}
   }
   n35 = Get Stuffing_Float_B()
   n36 = "Transform (Modify) Bone - Bone: dyn_stuffing_06_r"(ComponentPose=@n34, Alpha=0.75, Translation=@n35, Rotation=@n26, Scale="1,1,1") {
    rotationMode=BMM_Replace
    rotationSpace=BCS_BoneSpace
    translationMode=BMM_Additive
    translationSpace=BCS_BoneSpace
    alphaBoolBlend {blendOption=Linear}
    boneToModify {boneName=dyn_stuffing_06_r}
   }
   n37 = "Transform (Modify) Bone - Bone: dyn_stuffing_09_r"(ComponentPose=@n36, Alpha=0.5, Translation=@n35, Rotation="0,0,0", Scale="1,1,1") {
    rotationMode=BMM_Replace
    rotationSpace=BCS_BoneSpace
    translationMode=BMM_Additive
    translationSpace=BCS_BoneSpace
    alphaBoolBlend {blendOption=Linear}
    boneToModify {boneName=dyn_stuffing_09_r}
   }
   n38 = "Transform (Modify) Bone - Bone: dyn_glove_shred_r"(ComponentPose=@n37, Alpha=1, Translation=@n35, Rotation="0,0,0", Scale="1,1,1") {
    rotationMode=BMM_Replace
    rotationSpace=BCS_BoneSpace
    translationMode=BMM_Additive
    alphaBoolBlend {blendOption=Linear}
    boneToModify {boneName=dyn_glove_shred_r}
   }
   n39 = "Transform (Modify) Bone - Bone: dyn_stuffing_10_r"(ComponentPose=@n38, Alpha=1, Translation=@n35, Rotation="0,0,0", Scale="1,1,1") {
    rotationMode=BMM_Replace
    rotationSpace=BCS_BoneSpace
    translationMode=BMM_Additive
    alphaBoolBlend {blendOption=Linear}
    boneToModify {boneName=dyn_stuffing_10_r}
   }
   n40 = "Transform (Modify) Bone - Bone: dyn_stuffing_01_l"(ComponentPose=@n39, Alpha=0.5, Translation=@n24, Rotation="0,0,0", Scale="1,1,1") {
    rotationMode=BMM_Replace
    rotationSpace=BCS_BoneSpace
    translationMode=BMM_Additive
    translationSpace=BCS_BoneSpace
    alphaBoolBlend {blendOption=Linear}
    boneToModify {boneName=dyn_stuffing_01_l}
   }
   n41 = Get Spinning_Rotator()
   n42 = "Transform (Modify) Bone - Bone: dyn_stuffing_02_l"(ComponentPose=@n40, Alpha=0.5, Translation=@n24, Rotation=@n41, Scale="1,1,1") {
    rotationMode=BMM_Replace
    rotationSpace=BCS_BoneSpace
    translationMode=BMM_Additive
    translationSpace=BCS_BoneSpace
    alphaBoolBlend {blendOption=Linear}
    boneToModify {boneName=dyn_stuffing_02_l}
   }
   n43 = Get Stuffing_Float_C()
   n44 = "Transform (Modify) Bone - Bone: dyn_stuffing_04_l"(ComponentPose=@n42, Alpha=1, Translation=@n43, Rotation="0,0,0", Scale="1,1,1") {
    rotationMode=BMM_Replace
    rotationSpace=BCS_BoneSpace
    translationMode=BMM_Additive
    translationSpace=BCS_BoneSpace
    alphaBoolBlend {blendOption=Linear}
    boneToModify {boneName=dyn_stuffing_04_l}
   }
   n45 = "Transform (Modify) Bone - Bone: dyn_stuffing_05_l"(ComponentPose=@n44, Alpha=1, Translation=@n43, Rotation="0,0,0", Scale="1,1,1") {
    rotationMode=BMM_Replace
    rotationSpace=BCS_BoneSpace
    translationMode=BMM_Additive
    translationSpace=BCS_BoneSpace
    alphaBoolBlend {blendOption=Linear}
    boneToModify {boneName=dyn_stuffing_05_l}
   }
   n46 = "Transform (Modify) Bone - Bone: dyn_stuffing_06_l"(ComponentPose=@n45, Alpha=0.5, Translation=@n43, Rotation=@n41, Scale="1,1,1") {
    rotationMode=BMM_Replace
    rotationSpace=BCS_BoneSpace
    translationMode=BMM_Additive
    translationSpace=BCS_BoneSpace
    alphaBoolBlend {blendOption=Linear}
    boneToModify {boneName=dyn_stuffing_06_l}
   }
   n47 = Get Stuffing_Float_B()
   n48 = "Transform (Modify) Bone - Bone: dyn_glove_shred_l"(ComponentPose=@n46, Alpha=0.75, Translation=@n47, Rotation="0,0,0", Scale="1,1,1") {
    rotationMode=BMM_Replace
    rotationSpace=BCS_BoneSpace
    translationMode=BMM_Additive
    translationSpace=BCS_BoneSpace
    alphaBoolBlend {blendOption=Linear}
    boneToModify {boneName=dyn_glove_shred_l}
   }
   n49 = "Transform (Modify) Bone - Bone: dyn_stuffing_08_l"(ComponentPose=@n48, Alpha=0.5, Translation=@n47, Rotation=@n41, Scale="1,1,1") {
    rotationMode=BMM_Replace
    rotationSpace=BCS_BoneSpace
    translationMode=BMM_Additive
    translationSpace=BCS_BoneSpace
    alphaBoolBlend {blendOption=Linear}
    boneToModify {boneName=dyn_stuffing_08_l}
   }
   n50 = "Transform (Modify) Bone - Bone: dyn_stuffing_09_l"(ComponentPose=@n49, Alpha=0.5, Translation=@n47, Rotation="0,0,0", Scale="1,1,1") {
    rotationMode=BMM_Replace
    rotationSpace=BCS_BoneSpace
    translationMode=BMM_Additive
    translationSpace=BCS_BoneSpace
    alphaBoolBlend {blendOption=Linear}
    boneToModify {boneName=dyn_stuffing_09_l}
   }
   n51 = "Transform (Modify) Bone - Bone: dyn_stuffing_10_l"(ComponentPose=@n50, Alpha=1, Translation=@n47, Rotation="0,0,0", Scale="1,1,1") {
    rotationMode=BMM_Replace
    rotationSpace=BCS_BoneSpace
    translationMode=BMM_Additive
    translationSpace=BCS_BoneSpace
    alphaBoolBlend {blendOption=Linear}
    boneToModify {boneName=dyn_stuffing_10_l}
   }
   n52 = "Transform (Modify) Bone - Bone: dyn_stuffing_12_l"(ComponentPose=@n51, Alpha=0.3, Translation=@n47, Rotation=@n41, Scale="1,1,1") {
    rotationMode=BMM_Replace
    rotationSpace=BCS_BoneSpace
    translationMode=BMM_Additive
    alphaBoolBlend {blendOption=Linear}
    boneToModify {boneName=dyn_stuffing_12_l}
   }
   n53 = RigidBody(ComponentPose=@n52) {
    bClampLinearTranslationLimitToRefPose=true
    bOverrideWorldGravity=true
    overridePhysicsAsset=&/F_LRG_Bunny_Purple_Jacket_Physics
    simulationSpace=BaseBoneSpace
    alphaBoolBlend {blendOption=Linear}
    baseBoneRef {boneName=spine_05}
    componentAppliedLinearAccClamp=<1000 1000 1000>
    componentLinearAccScale=<0.25 0.25 0.4>
    componentLinearVelScale=<0.25 0.25 0.4>
    overrideWorldGravity {z=-980}
   }
   n54 = RigidBody(ComponentPose=@n53) {
    bClampLinearTranslationLimitToRefPose=true
    bOverrideWorldGravity=true
    overridePhysicsAsset=&/F_LRG_Bunny_Purple_Jacket_Frontend_Physics
    simulationSpace=BaseBoneSpace
    alphaBoolBlend {blendOption=Linear}
    baseBoneRef {boneName=spine_05}
    componentAppliedLinearAccClamp=<1000 1000 1000>
    componentLinearAccScale=<0.25 0.25 0.4>
    componentLinearVelScale=<0.25 0.25 0.4>
    overrideWorldGravity {z=-980}
   }
   n55 = RigidBody(ComponentPose=@n54, Alpha=Get IsWearingBackbling_Alpha() -> (IsWearingBackbling_Alpha:double)) {
    bClampLinearTranslationLimitToRefPose=true
    bOverrideWorldGravity=true
    overridePhysicsAsset=&/F_LRG_Bunny_Purple_Hood_Physics
    simulationSpace=BaseBoneSpace
    alphaBoolBlend {blendOption=Linear}
    baseBoneRef {boneName=spine_05}
    componentAppliedLinearAccClamp=<1000 1000 1000>
    componentLinearAccScale=<0.35 0.35 0.5>
    componentLinearVelScale=<0.35 0.35 0.5>
    overrideWorldGravity {z=-980}
   }
   n56 = RigidBody(ComponentPose=@n55) {
    bClampLinearTranslationLimitToRefPose=true
    bOverrideWorldGravity=true
    overridePhysicsAsset=&/F_LRG_Bunny_Purple_Skirt_Physics
    simulationSpace=BaseBoneSpace
    alphaBoolBlend {blendOption=Linear}
    baseBoneRef {boneName=pelvis}
    componentAppliedLinearAccClamp=<1000 1000 1000>
    componentLinearAccScale=<0.25 0.25 0.25>
    componentLinearVelScale=<0.25 0.25 0.25>
    overrideWorldGravity {z=-980}
    simSpaceSettings {
     maxAngularAcceleration=1000
     maxAngularVelocity=3
     maxLinearAcceleration=700
     maxLinearVelocity=500
     velocityScaleZ=0
     worldAlpha=1
     externalLinearDragV=<0.2 0.2>
    }
   }
   n57 = "Transform (Modify) Bone - Bone: dyn_keychain_base"(ComponentPose=@n56, Alpha=Get Crouch_Alpha() -> (Crouch_Alpha:double), Translation="3,4.5,2.5", Rotation="-5,0,-40", Scale="1,1,1") {
    rotationMode=BMM_Additive
    rotationSpace=BCS_BoneSpace
    translationMode=BMM_Additive
    translationSpace=BCS_BoneSpace
    alphaBoolBlend {blendOption=Linear}
    alphaScaleBiasClamp {bInterpResult=true, interpSpeedDecreasing=4, interpSpeedIncreasing=6}
    boneToModify {boneName=dyn_keychain_base}
    rotation {pitch=-5, roll=-40}
    translation=<3 4.5 2.5>
   }
   n58 = Anim Dynamics - Bone: dyn_zipper(ComponentPose=@n57, Alpha=0.5) {
    angularDampingOverride=0.9999
    bGravityOverrideInSimSpace=true
    bOverrideAngularDamping=true
    bOverrideLinearDamping=true
    linearDampingOverride=0.9999
    alphaBoolBlend {blendOption=Linear}
    boundBone {boneName=dyn_zipper}
    componentAppliedLinearAccClamp=<6000 6000 6000>
    componentLinearAccScale=<0.2 0.2 0.2>
    componentLinearVelScale=<0.2 0.2 0.2>
    physicsBodyDefinitions [{
      sphereCollisionRadius=0
      boundBone {boneName=dyn_zipper}
      boxExtents=<12 12 12>
      constraintSetup {angularLimitsMax=<0 10 30>, angularLimitsMin=<0 -10 -30>, angularTarget=<0 0 0>}
      localJointOffset=<2 0 0>
     }]
    relativeSpaceBone {boneName=spine_05}
   }
   n59 = RigidBody(ComponentPose=@n58) {
    bClampLinearTranslationLimitToRefPose=true
    bOverrideWorldGravity=true
    overridePhysicsAsset=&F_LRG_Bunny_Purple_Ears_Physics
    simulationSpace=BaseBoneSpace
    alphaBoolBlend {blendOption=Linear}
    baseBoneRef {boneName=spine_05}
    componentAppliedLinearAccClamp=<6000 6000 6000>
    componentLinearAccScale=<0.25 0.25 0.5>
    componentLinearVelScale=<0.25 0.25 0.5>
    overrideWorldGravity {z=-980}
   }
   n60 = Anim Dynamics - Bone: dyn_bunny_head_ear_l(ComponentPose=@n59, Alpha=1, ExternalForce="0,0,0") {
    angularBiasOverride=0.96
    angularDampingOverride=0.98
    bGravityOverrideInSimSpace=true
    bOverrideAngularBias=true
    bOverrideAngularDamping=true
    bUseGravityOverride=true
    gravityScale=0.25
    linearDampingOverride=0.98
    simulationSpace=BoneRelative
    alphaBoolBlend {blendOption=Linear}
    boundBone {boneName=dyn_bunny_head_ear_l}
    componentAppliedLinearAccClamp=<500 500 500>
    componentLinearAccScale=<0.1 0.25 0.5>
    componentLinearVelScale=<0.1 0.25 0.5>
    gravityOverride {z=-600}
    physicsBodyDefinitions [{
      sphereCollisionRadius=0
      boundBone {boneName=dyn_bunny_head_ear_l}
      boxExtents=<10 15 10>
      constraintSetup {angularLimitsMax=<2 -10 25>, angularLimitsMin=<-2 -10 -25>, angularTarget=<0 0 0>}
      localJointOffset=<5 0 0>
     }]
    relativeSpaceBone {boneName=root}
   }
   n61 = Anim Dynamics - Bone: dyn_bunny_head_ear_01_l(ComponentPose=@n60, Alpha=1, ExternalForce="0,0,0") {
    angularBiasOverride=0.96
    angularDampingOverride=0.98
    bGravityOverrideInSimSpace=true
    bOverrideAngularBias=true
    bOverrideAngularDamping=true
    bUseGravityOverride=true
    gravityScale=0.25
    linearDampingOverride=0.98
    simulationSpace=BoneRelative
    alphaBoolBlend {blendOption=Linear}
    boundBone {boneName=dyn_bunny_head_ear_01_l}
    componentAppliedLinearAccClamp=<500 500 500>
    componentLinearAccScale=<0.1 0.25 0.5>
    componentLinearVelScale=<0.1 0.25 0.5>
    gravityOverride {z=-600}
    physicsBodyDefinitions [{
      sphereCollisionRadius=0
      boundBone {boneName=dyn_bunny_head_ear_01_l}
      boxExtents=<10 15 10>
      constraintSetup {angularLimitsMax=<2 -10 25>, angularLimitsMin=<-2 -10 -25>, angularTarget=<0 0 0>}
      localJointOffset=<5 0 0>
     }]
    relativeSpaceBone {boneName=root}
   }
   n62 = Anim Dynamics - Bone: dyn_bunny_head_ear_r(ComponentPose=@n61, Alpha=1, ExternalForce="0,0,0") {
    angularBiasOverride=0.96
    angularDampingOverride=0.98
    bGravityOverrideInSimSpace=true
    bOverrideAngularBias=true
    bOverrideAngularDamping=true
    bUseGravityOverride=true
    gravityScale=0.25
    linearDampingOverride=0.98
    simulationSpace=BoneRelative
    alphaBoolBlend {blendOption=Linear}
    boundBone {boneName=dyn_bunny_head_ear_r}
    componentAppliedLinearAccClamp=<500 500 500>
    componentLinearAccScale=<0.1 0.25 0.5>
    componentLinearVelScale=<0.1 0.25 0.5>
    gravityOverride {z=-600}
    physicsBodyDefinitions [{
      sphereCollisionRadius=0
      boundBone {boneName=dyn_bunny_head_ear_r}
      boxExtents=<10 15 10>
      constraintSetup {angularLimitsMax=<2 20 25>, angularLimitsMin=<-2 -20 -25>, angularTarget=<0 0 0>}
      localJointOffset=<7 0 0>
     }]
    relativeSpaceBone {boneName=root}
   }
   n63 = Anim Dynamics - Bone: dyn_bunny_head_ear_01_r(ComponentPose=@n62, Alpha=1, ExternalForce="0,0,0") {
    angularBiasOverride=0.96
    angularDampingOverride=0.98
    bGravityOverrideInSimSpace=true
    bOverrideAngularBias=true
    bOverrideAngularDamping=true
    bUseGravityOverride=true
    gravityScale=0.25
    linearDampingOverride=0.98
    simulationSpace=BoneRelative
    alphaBoolBlend {blendOption=Linear}
    boundBone {boneName=dyn_bunny_head_ear_01_r}
    componentAppliedLinearAccClamp=<500 500 500>
    componentLinearAccScale=<0.1 0.25 0.5>
    componentLinearVelScale=<0.1 0.25 0.5>
    gravityOverride {z=-600}
    physicsBodyDefinitions [{
      sphereCollisionRadius=0
      boundBone {boneName=dyn_bunny_head_ear_01_r}
      boxExtents=<10 15 10>
      constraintSetup {angularLimitsMax=<2 20 25>, angularLimitsMin=<-2 -20 -25>, angularTarget=<0 0 0>}
      localJointOffset=<7 0 0>
     }]
    relativeSpaceBone {boneName=root}
   }
   n64 = "Transform (Modify) Bone - Bone: dyn_bunny_head_base"(ComponentPose=@n63, Alpha=Get Skydive_Alpha() -> (Skydive_Alpha:double), Translation="0.000,8,0", Rotation="75,30,25", Scale="1,1,1") {
    rotationMode=BMM_Additive
    rotationSpace=BCS_BoneSpace
    translationMode=BMM_Additive
    translationSpace=BCS_BoneSpace
    alphaBoolBlend {blendOption=Linear}
    alphaScaleBiasClamp {bInterpResult=true, interpSpeedDecreasing=4, interpSpeedIncreasing=6}
    boneToModify {boneName=dyn_bunny_head_base}
    rotation=<rot 75 30 25>
    translation {y=8}
   }
   n65 = "Transform (Modify) Bone - Bone: dyn_stuffing_02_base_l"(ComponentPose=@n64, Alpha=1, Translation="0.000,-3,-4", Rotation="0,0,0", Scale="1,1,1") {
    translationMode=BMM_Additive
    translationSpace=BCS_BoneSpace
    alphaBoolBlend {blendOption=Linear}
    boneToModify {boneName=dyn_stuffing_02_base_l}
    translation {y=-3, z=-4}
   }
   n66 = Control Rig!(Source=Component To Local(ComponentPose=@n65), Alpha, crouch still:double=Get Crouch_STill() -> (Crouch_STill:double)) -> (Pose) {
    alphaBoolBlend {blendOption=Linear}
    controlRigAssetReference {blueprintRigClass=&/F_LRG_Bunny_Purple_CtrlRig.F_LRG_Bunny_Purple_CtrlRig_C}
   }
   n67 = "Constraint - wrist_r (PPPPPP)"!(
    ComponentPose=Local To Component(LocalPose=@n66)
    Alpha=1
    ConstraintWeights_0=0.0985
    ConstraintWeights_1:float=0.1177
    ConstraintWeights_2:float=0.0024
    ConstraintWeights_3:float=0.5216
    ConstraintWeights_4:float=0.2445
    ConstraintWeights_5:float=0.0153
   ) -> (Pose) {
    alphaBoolBlend {blendOption=Linear}
    boneToModify {boneName=wrist_r}
    constraintSetup [
     {targetBone {boneName=spine_02}}
     {targetBone {boneName=spine_01}}
     {targetBone {boneName=spine_05}}
     {targetBone {boneName=pelvis}}
     {targetBone {boneName=spine_03}}
     {targetBone {boneName=thigh_twist_01_r}}
    ]
    constraintWeights [1,1,1,1,1,1]
   }
   n68 = "Constraint - wrist_l (PPPPPPPP)"!(
    ComponentPose=@n67
    Alpha=1
    ConstraintWeights_0=0.0006
    ConstraintWeights_1:float=0.0848
    ConstraintWeights_2:float=0.0036
    ConstraintWeights_3:float=0.1238
    ConstraintWeights_4:float=0.0013
    ConstraintWeights_5:float=0.0021
    ConstraintWeights_6:float=0.5567
    ConstraintWeights_7:float=0.2271
   ) -> (Pose) {
    alphaBoolBlend {blendOption=Linear}
    boneToModify {boneName=wrist_l}
    constraintSetup [
     {targetBone {boneName=thigh_l}}
     {targetBone {boneName=spine_02}}
     {targetBone {boneName=thigh_twist_01_l}}
     {targetBone {boneName=spine_01}}
     {targetBone {boneName=spine_04}}
     {targetBone {boneName=spine_05}}
     {targetBone {boneName=pelvis}}
     {targetBone {boneName=spine_03}}
    ]
    constraintWeights [1,1,1,1,1,1,1,1]
   }
   n69 = "Constraint - dyn_zipper_base (PPPPP)"!(
    ComponentPose=@n68
    Alpha=1
    ConstraintWeights_0=0.197
    ConstraintWeights_1:float=0.0207
    ConstraintWeights_2:float=0.637
    ConstraintWeights_3:float=0.0811
    ConstraintWeights_4:float=0.0642
   ) -> (Pose) {
    alphaBoolBlend {blendOption=Linear}
    boneToModify {boneName=dyn_zipper_base}
    constraintSetup [
     {targetBone {boneName=spine_04}}
     {targetBone {boneName=neck_01}}
     {targetBone {boneName=spine_05}}
     {targetBone {boneName=clavicle_r}}
     {targetBone {boneName=clavicle_l}}
    ]
    constraintWeights [1,1,1,1,1]
   }
   n70 = Get Pickaxe_Equipped() -> (Pickaxe_Equipped:bool)
   n71 = "Transform (Modify) Bone - Bone: dyn_stuffing_05_base_r"!(ComponentPose=@n69, bAlphaBoolEnabled:bool=@n70, Translation="0.000,0.000,-3", Rotation="0,0,0", Scale="1,1,1") -> (Pose) {
    alphaInputType=Bool
    rotationMode=BMM_Additive
    rotationSpace=BCS_BoneSpace
    translationMode=BMM_Additive
    translationSpace=BCS_BoneSpace
    alphaBoolBlend {blendOption=Linear}
    alphaScaleBiasClamp {bInterpResult=true, interpSpeedIncreasing=200}
    boneToModify {boneName=dyn_stuffing_05_base_r}
    translation {z=-3}
   }
   n72 = "Transform (Modify) Bone - Bone: dyn_stuffing_01_base_r"!(ComponentPose=@n71, bAlphaBoolEnabled:bool=@n70, Translation="0.000,0.000,-3", Rotation="0,0,0", Scale="1,1,1") -> (Pose) {
    alphaInputType=Bool
    rotationMode=BMM_Additive
    rotationSpace=BCS_BoneSpace
    translationMode=BMM_Additive
    translationSpace=BCS_BoneSpace
    alphaBoolBlend {blendOption=Linear}
    alphaScaleBiasClamp {bInterpResult=true, interpSpeedIncreasing=200}
    boneToModify {boneName=dyn_stuffing_01_base_r}
    translation {z=-3}
   }
   n73 = "Transform (Modify) Bone - Bone: dyn_stuffing_07_base_r"!(ComponentPose=@n72, bAlphaBoolEnabled:bool=@n70, Translation="0.000,0.000,-5", Rotation="0,0,0", Scale="1,1,1") -> (Pose) {
    alphaInputType=Bool
    rotationMode=BMM_Additive
    rotationSpace=BCS_BoneSpace
    translationMode=BMM_Additive
    translationSpace=BCS_BoneSpace
    alphaBoolBlend {blendOption=Linear}
    alphaScaleBiasClamp {bInterpResult=true, interpSpeedIncreasing=200}
    boneToModify {boneName=dyn_stuffing_07_base_r}
    translation {z=-5}
   }
   n74 = "Transform (Modify) Bone - Bone: dyn_stuffing_09_base_r"!(ComponentPose=@n73, bAlphaBoolEnabled:bool=@n70, Translation="0.000,-5,4", Rotation="0.000,30,0.000", Scale="1,1,1") -> (Pose) {
    alphaInputType=Bool
    rotationMode=BMM_Additive
    rotationSpace=BCS_BoneSpace
    translationMode=BMM_Additive
    translationSpace=BCS_BoneSpace
    alphaBoolBlend {blendOption=Linear}
    alphaScaleBiasClamp {bInterpResult=true, interpSpeedIncreasing=200}
    boneToModify {boneName=dyn_stuffing_09_base_r}
    rotation {yaw=30}
    translation {y=-5, z=4}
   }
   n75 = "Transform (Modify) Bone - Bone: dyn_stuffing_08_base_r"!(ComponentPose=@n74, bAlphaBoolEnabled:bool=@n70, Translation="4,0.000,-0.000", Rotation="0,0,0", Scale="1,1,1") -> (Pose) {
    alphaInputType=Bool
    rotationMode=BMM_Additive
    rotationSpace=BCS_BoneSpace
    translationMode=BMM_Additive
    translationSpace=BCS_BoneSpace
    alphaBoolBlend {blendOption=Linear}
    alphaScaleBiasClamp {bInterpResult=true, interpSpeedIncreasing=200}
    boneToModify {boneName=dyn_stuffing_08_base_r}
    translation {x=4}
   }
   n76 = "Transform (Modify) Bone - Bone: dyn_glove_shred_base_r"!(ComponentPose=@n75, bAlphaBoolEnabled:bool=@n70, Translation="0,1,-2", Rotation="60,-20,0", Scale="1,1,1") -> (Pose) {
    alphaInputType=Bool
    rotationMode=BMM_Additive
    rotationSpace=BCS_BoneSpace
    translationMode=BMM_Additive
    translationSpace=BCS_BoneSpace
    alphaBoolBlend {blendOption=Linear}
    alphaScaleBiasClamp {bInterpResult=true, interpSpeedIncreasing=200}
    boneToModify {boneName=dyn_glove_shred_base_r}
    rotation {pitch=60, yaw=-20}
    translation {y=1, z=-2}
   }
   n77 = "Transform (Modify) Bone - Bone: dyn_stuffing_04_base_r"!(ComponentPose=@n76, bAlphaBoolEnabled:bool=@n70, Translation="0,0,-4", Rotation="0,0,0", Scale="1,1,1") -> (Pose) {
    alphaInputType=Bool
    rotationMode=BMM_Additive
    rotationSpace=BCS_BoneSpace
    translationMode=BMM_Additive
    translationSpace=BCS_BoneSpace
    alphaBoolBlend {blendOption=Linear}
    alphaScaleBiasClamp {bInterpResult=true, interpSpeedIncreasing=200}
    boneToModify {boneName=dyn_stuffing_04_base_r}
    translation {z=-4}
   }
   n78 = "Transform (Modify) Bone - Bone: dyn_stuffing_10_base_r"!(ComponentPose=@n77, bAlphaBoolEnabled:bool=@n70, Translation="0,0,0", Rotation="40,60,90", Scale="1,1,1") -> (Pose) {
    alphaInputType=Bool
    rotationMode=BMM_Additive
    rotationSpace=BCS_BoneSpace
    translationMode=BMM_Additive
    translationSpace=BCS_BoneSpace
    alphaBoolBlend {blendOption=Linear}
    alphaScaleBiasClamp {bInterpResult=true, interpSpeedIncreasing=200}
    boneToModify {boneName=dyn_stuffing_10_base_r}
    rotation=<rot 40 60 90>
   }
   n79 = "Transform (Modify) Bone - Bone: dyn_stuffing_02_l"!(
    ComponentPose=@n78
    bAlphaBoolEnabled:bool=Get rifle_Equipped() -> (rifle_Equipped:bool)
    Translation="2,-4,4"
    Rotation="0,0,0"
    Scale="1,1,1"
   ) -> (Pose) {
    alphaInputType=Bool
    rotationMode=BMM_Additive
    rotationSpace=BCS_BoneSpace
    translationMode=BMM_Additive
    translationSpace=BCS_BoneSpace
    alphaBoolBlend {blendOption=Linear}
    alphaScaleBiasClamp {bInterpResult=true, interpSpeedIncreasing=200}
    boneToModify {boneName=dyn_stuffing_02_l}
    translation=<2 -4 4>
   }
   n80 = Get bIsInFrontEnd()
   n81 = "Transform (Modify) Bone - Bone: dyn_stuffing_04_base_l"!(ComponentPose=@n79, bAlphaBoolEnabled:bool=@n80, Translation="10,-5.5,5", Rotation="80,0.000,0.000", Scale="1,1,1") -> (Pose) {
    alphaInputType=Bool
    rotationMode=BMM_Additive
    rotationSpace=BCS_BoneSpace
    scaleSpace=BCS_BoneSpace
    translationMode=BMM_Additive
    translationSpace=BCS_BoneSpace
    alphaBoolBlend {blendOption=Linear}
    boneToModify {boneName=dyn_stuffing_04_base_l}
    rotation {pitch=80}
    translation=<10 -5.5 5>
   }
   n82 = "Transform (Modify) Bone - Bone: dyn_stuffing_10_l"!(ComponentPose=@n81, bAlphaBoolEnabled:bool=@n80, Translation="0,2,0", Rotation="-20,0.000,0.000", Scale="1,1,1") -> (Pose) {
    alphaInputType=Bool
    rotationMode=BMM_Replace
    rotationSpace=BCS_BoneSpace
    translationMode=BMM_Additive
    translationSpace=BCS_BoneSpace
    alphaBoolBlend {blendOption=Linear}
    boneToModify {boneName=dyn_stuffing_10_l}
    rotation {pitch=-20}
    translation {y=2}
   }
   n83 = "Transform (Modify) Bone - Bone: dyn_glove_shred_base_r"(ComponentPose=@n82, bAlphaBoolEnabled=@n80, Translation="0,0,-1", Rotation="0,0,0", Scale="1,1,1") {
    alphaInputType=Bool
    rotationMode=BMM_Replace
    rotationSpace=BCS_BoneSpace
    translationMode=BMM_Additive
    translationSpace=BCS_BoneSpace
    alphaBoolBlend {blendOption=Linear}
    boneToModify {boneName=dyn_glove_shred_base_r}
    translation {z=-1}
   }
   Save cached pose 'CachePose'(Pose=Component To Local(ComponentPose=@n83)) {cacheName=CachePose}
  ]}
 EventGraph {nodes [
   Event BlueprintInitializeAnimation()
   > n1 = Cast To PlayerPawn_Athena(Object=TryGetPawnOwner())
   > n2 = Set Pawn Owner(Pawn Owner:Player Pawn Athena Object Reference=@n1) -> (Output_Get:Player Pawn Athena Object Reference)
   > Set Body_Mesh(Body_Mesh=GetSkeletalMeshForPartType(self=@n2))
   Event OnCharacterCustomizationLoaded()
   > Do Once()
   n3 = CheckForBackbling()
   > n4 = Set IsWearingBackbling(IsWearingBackbling:bool=OR Boolean!(A=@n3, B=@n3.bIsBackblingACape, C:bool=Get bIsInFrontEnd()) -> (ReturnValue)) -> (Output_Get:bool)
   > n5 = Branch(Condition=@n4)
   > Set IsWearingBackbling_Alpha(IsWearingBackbling_Alpha:double=0.25) -> (Output_Get:double)
   Set IsWearingBackbling_Alpha(execute=@n5.else, IsWearingBackbling_Alpha=1)
   Event BlueprintUpdateAnimation()
   > n6 = Sequence!(execute) -> (then_0, then_1, then_2:Exec, then_3:Exec, then_4:Exec, then_5:Exec, then_6:Exec)
   n7 = GetRealTimeSeconds()
   n8 = "Sin (Radians)"(A=float * float(A=float * float(A=1, B=1), B=@n7))
   n9 = "Sin (Radians)"(A=float * float(A=float * float(A=1, B=2), B=@n7))
   n10 = "Sin (Radians)"(A=float * float(A=float * float(A=1, B=1.5), B=@n7))
   Set Stuffing_Float_A(
    execute=@n6.then_1
    Stuffing_Float_A:Vector=MakeVector(X=float * float(A=@n10, B=0.3), Y=float * float(A=@n8, B=0.56), Z=float * float(A=@n9, B=0.22))
   ) -> (Output_Get:Vector)
   n11 = GetGameTimeInSeconds()
   n12 = "Cos (Radians)"!(A:double=float * float(A=float * float(A=1, B=1.5), B=@n11)) -> (ReturnValue:double)
   n13 = "Cos (Radians)"(A=float * float(A=float * float(A=1, B=1.78), B=@n11))
   n14 = "Cos (Radians)"(A=float * float(A=float * float(A=1, B=1.15), B=@n11))
   > Set Stuffing_Float_B(Stuffing_Float_B:Vector=MakeVector(X=float * float(A=@n12, B=0.6), Y=float * float(A=@n13, B=0.725), Z=float * float(A=@n14, B=0.3))) -> (Output_Get:Vector)
   n15 = GetGameTimeInSeconds()
   n16 = float * float(A=float * float(A=1, B=1.5), B=@n15)
   n17 = float * float(A=float * float(A=1, B=1), B=@n15)
   n18 = float * float(A=float * float(A=1, B=0.1), B=@n15)
   n19 = MakeVector(
    X=float * float(A="Sin (Radians)"(A=@n16), B=0.3)
    Y=float * float(A="Sin (Radians)"(A=@n17), B=0.1)
    Z=float * float(A="Sin (Radians)"(A=@n18), B=0.25)
   )
   > Set Stuffing_Float_C(Stuffing_Float_C:Vector=@n19) -> (Output_Get:Vector)
   n20 = BreakVector(InVec=vector * float(A=@n19, B=7))
   > Set Floating_Rotator(Floating_Rotator:Rotator=MakeRotator(Roll=@n20, Pitch=@n20.Y, Yaw=@n20.Z)) -> (Output_Get:Rotator)
   n21 = Get bIsCrouching()
   Set Crouch_Alpha(execute=@n6.then_2, Crouch_Alpha:double=SelectFloat(A=1, B=0, bPickA=@n21)) -> (Output_Get:double)
   n22 = NOT Boolean(A=Get bIsMoving2D())
   > Set Crouch_STill(Crouch_STill:double="To Float (Boolean)"(InBool=AND Boolean(A=@n21, B=@n22))) -> (Output_Get:double)
   n23 = Get bIsInFrontEnd()
   > Set Front End Off(Front End Off:double=SelectFloat(A=1, B=0, bPickA=NOT Boolean(A=@n23))) -> (Output_Get:double)
   n24 = float * float(A=float / float(A=GetRealTimeSeconds(), B=-720) {is=title K2Node_CallFunction}, B=10000)
   Set Spinning_Rotator(execute=@n6.then_0, Spinning_Rotator:Rotator=MakeRotator(Roll=@n24, Yaw=@n24)) -> (Output_Get:Rotator)
   Set Head_Rot_Back(
    execute=@n6.then_5
    Head_Rot_Back:double=GetAlphaFromReferenceTransformDelta(
     SkeletalMeshComponent=Get Body_Mesh()
     BoneName=head
     BaseName=spine_05
     DeltaComponentType=QuaternionTwist
     InputRange=(X=-48,Y=0)
     TwistAxis="0,0,1"
    )
   ) -> (Output_Get:double)
   Set Pickaxe_Equipped(execute=@n6.then_6, Pickaxe_Equipped:bool=false) -> (Output_Get:bool)
   > Set rifle_Equipped(rifle_Equipped:bool=false) -> (Output_Get:bool)
   > n25 = Switch on EFortWeaponCoreAnimation!(
    execute
    Selection:EFortWeaponCoreAnimation Enum=GetWeaponCoreAnim!(self:Fort Player Animation Interface Object Reference=GetOwnerAnimationInterface!(self:Custom Character Part Anim Instance Object Reference) -> (ReturnValue:Fort Player Animation Interface Object Reference)) -> (ReturnValue:EFortWeaponCoreAnimation Enum)
   ) -> (
    Melee:Exec
    Pistol:Exec
    Shotgun:Exec
    PaperBlueprint:Exec
    Rifle:Exec
    MeleeOneHand:Exec
    MachinePistol:Exec
    RocketLauncher:Exec
    GrenadeLauncher:Exec
    GoingCommando:Exec
    AssaultRifle:Exec
    TacticalShotgun:Exec
    SniperRifle:Exec
    TrapPlacement:Exec
    ShoulderLauncher:Exec
    AbilityDecoTool:Exec
    Crossbow:Exec
    C4:Exec
    RemoteControl:Exec
    DualWield:Exec
    AR_BullPup:Exec
    AR_ForwardGrip:Exec
    MedPackPaddles:Exec
    SMG_P90:Exec
    AR_DrumGun:Exec
    Consumable_Small:Exec
    Consumable_Large:Exec
    Balloon:Exec
    MountedTurret:Exec
    CreativeTool:Exec
    ExplosiveBow:Exec
    AshtonIndigo:Exec
    AshtonChicago:Exec
    MeleeDualWield:Exec
    Unarmed:Exec
   )
   Set rifle_Equipped(
    execute=[
     @n25.AR_BullPup
     @n25.AR_DrumGun
     @n25.AR_ForwardGrip
     @n25.AssaultRifle
     @n25.Crossbow
     @n25.GrenadeLauncher
     @n25.Rifle
     @n25.Shotgun
     @n25.SniperRifle
     @n25.TacticalShotgun
    ]
    rifle_Equipped=true
   )
   Set Pickaxe_Equipped(execute=@n25.Melee, Pickaxe_Equipped=true)
   Set Skydive_Alpha(execute=@n6.then_3, Skydive_Alpha:double=SelectFloat(A=1, B=0, bPickA=Get bIsSkyDiving())) -> (Output_Get:double)
   Set ADS_Alpha(execute=@n6.then_4, ADS_Alpha:double=SelectFloat(A=1, B=0, bPickA=Get bIsTargeting())) -> (Output_Get:double)
   n26 = BreakTransform(InTransform=GetDeltaTransformFromRefPose(self=GetOwnerSkeletalMeshComponent(), BoneName=head, BaseName=spine_05))
   Set Head_Rot_Back(Head_Rot_Back=MapRangeClamped(Value=BreakRotator(InRot=@n26.Rotation).Yaw, InRangeA=0, InRangeB=-48))
  ]}
]
variables [
 ADS_Alpha {type=number}
 Body_Mesh {type=SkeletalMeshComponent}
 Crouch_Alpha {type=number}
 Crouch_STill {type=number}
 Floating_Rotator {type=Rotator}
 Front End Off {type=number}
 Head_Rot_Back {type=number}
 IsWearingBackbling {type=boolean}
 IsWearingBackbling_Alpha {type=number}
 Pawn Owner {type=PlayerPawn_Athena_C}
 Pickaxe_Equipped {type=boolean}
 rifle_Equipped {type=boolean}
 Skydive_Alpha {type=number}
 Spinning_Rotator {type=Rotator}
 Stuffing_Float_A {type=Vector}
 Stuffing_Float_B {type=Vector}
 Stuffing_Float_C {type=Vector}
]
