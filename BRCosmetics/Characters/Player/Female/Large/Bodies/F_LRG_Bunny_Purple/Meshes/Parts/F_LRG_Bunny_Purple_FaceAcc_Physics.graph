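# Physics asset for the face accessory part of the large female Purple Bunny character body, covering its short dynamic hair bone chain.
# A small sphere on dyn_hair_base roots two capsules, dyn_hair and dyn_hair_01, linked in series by constraints with swing and twist limits that loosen toward the tip.
# Larger capsules on neck_01 and spine_05 have no constraints and appear to serve as collision volumes for the hair.
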PhysicsAsset /BRCosmetics/Characters/Player/Female/Large/Bodies/F_LRG_Bunny_Purple/Meshes/Parts/F_LRG_Bunny_Purple_FaceAcc_Physics
bodies {
 dyn_hair [{shapeType=Capsule, length=2.1383, radius=1.4515, shapeName=dyn_hair_capsule, center=<1.8624 0 0>, rotation=<rot 90 0 0>}]
 dyn_hair_01 [{shapeType=Capsule, length=1.206, radius=1.5266, shapeName=dyn_hair_01_capsule, center=<1.6642 0 0>, rotation=<rot 90 0 0>}]
 dyn_hair_base [{shapeType=Sphere, radius=0.5, shapeName=dyn_hair_base_sphere}]
 neck_01 [{
   shapeType=Capsule
   length=13.9617
   radius=4.2883
   shapeName=neck_01_capsule
   center=<5.8932 2.9629 5.4084>
   rotation=<rot 4.7112 -9.6185 -42.4156>
  }]
 spine_05 [{shapeType=Capsule, length=15, radius=15, shapeName=spine_05_capsule, center=<-7.5151 0 0.0002>, rotation=<rot 90 0 0>}]
}
constraints [
 {bone1Name=dyn_hair, bone2Name=dyn_hair_base, swing1LimitDegrees=20, swing2LimitDegrees=20, twistLimitDegrees=5}
 {bone1Name=dyn_hair_01, bone2Name=dyn_hair, swing1LimitDegrees=25, swing2LimitDegrees=25, twistLimitDegrees=10}
]
