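# Post-copy physics AnimBlueprint for the face accessory part of the large female purple Bunny character.
# It copies the attached parent mesh's pose, then runs RigidBody simulation in spine_05 space with a physics asset override.
# Partially blended, spring-driven Anim Dynamics add jiggle to the left and right bun bones, whose base bones are then constrained to the head.
# The event graph holds no logic.
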AnimBlueprint /BRCosmetics/Characters/Player/Female/Large/Bodies/F_LRG_Bunny_Purple/Meshes/Parts/F_LRG_Bunny_Purple_FaceAcc_AnimBP
target_skeleton=&/F_LRG_Bunny_Purple_FaceAcc_Skeleton
graphs [
 AnimGraph {nodes [
   n1 = RigidBody(ComponentPose=Local To Component(LocalPose=Copy Pose From Mesh() {bUseAttachedParent=true})) {
    bClampLinearTranslationLimitToRefPose=true
    bOverrideWorldGravity=true
    overridePhysicsAsset=&/F_LRG_Bunny_Purple_FaceAcc_Physics
    simulationSpace=BaseBoneSpace
    alphaBoolBlend {blendOption=Linear}
    baseBoneRef {boneName=spine_05}
    componentAppliedLinearAccClamp=<6000 6000 6000>
    componentLinearAccScale=<0.2 0.2 0.2>
    componentLinearVelScale=<0.2 0.2 0.2>
    overrideWorldGravity {z=-980}
   }
   n2 = Anim Dynamics - Bone: dyn_bun_l(ComponentPose=@n1, Alpha=0.5, ExternalForce="0,0,0") {
    angularDampingOverride=0.9999
    bAngularSpring=true
    bLinearSpring=true
    bOverrideAngularDamping=true
    bUseGravityOverride=true
    linearSpringConstant=50
    simulationSpace=Actor
    alphaBoolBlend {blendOption=Linear}
    boundBone {boneName=dyn_bun_l}
    componentLinearAccScale=<1.5 1.5 1.5>
    componentLinearVelScale=<1.5 1.5 1.5>
    gravityOverride=<0.9 0.9 0.9>
    physicsBodyDefinitions [{
      sphereCollisionRadius=0
      boundBone {boneName=dyn_bun_l}
      boxExtents=<6 6 6>
      constraintSetup {angularLimitsMax=<8 8 8>, angularLimitsMin=<-8 -8 -8>, angularTarget=<0 0 0>}
      localJointOffset=<2 0 0>
     }]
    relativeSpaceBone {boneName=spine_05}
   }
   n3 = Anim Dynamics - Bone: dyn_bun_r(ComponentPose=@n2, Alpha=0.6, ExternalForce="0,0,0") {
    angularDampingOverride=0.9999
    bAngularSpring=true
    bLinearSpring=true
    bOverrideAngularDamping=true
    bUseGravityOverride=true
    linearSpringConstant=50
    simulationSpace=Actor
    alphaBoolBlend {blendOption=Linear}
    boundBone {boneName=dyn_bun_r}
    componentLinearAccScale=<1.5 1.5 1.5>
    componentLinearVelScale=<1.5 1.5 1.5>
    gravityOverride=<0.9 0.9 0.9>
    physicsBodyDefinitions [{
      sphereCollisionRadius=0
      boundBone {boneName=dyn_bun_r}
      boxExtents=<6 6 6>
      constraintSetup {angularLimitsMax=<8 8 8>, angularLimitsMin=<-8 -8 -8>, angularTarget=<0 0 0>}
      localJointOffset=<2 0 0>
     }]
    relativeSpaceBone {boneName=spine_05}
   }
   n4 = "Constraint - dyn_bun_l_base (P)"(ComponentPose=@n3, Alpha=1, ConstraintWeights_0=1) {
    alphaBoolBlend {blendOption=Linear}
    boneToModify {boneName=dyn_bun_l_base}
    constraintSetup [{targetBone {boneName=head}}]
    constraintWeights [1]
   }
   n5 = "Constraint - dyn_bun_r_base (P)"(ComponentPose=@n4, Alpha=1, ConstraintWeights_0=1) {
    alphaBoolBlend {blendOption=Linear}
    boneToModify {boneName=dyn_bun_r_base}
    constraintSetup [{targetBone {boneName=head}}]
    constraintWeights [1]
   }
   n6 = "Transform (Modify) Bone - Bone: dyn_hair_base"!(ComponentPose=@n5, Alpha=1, Rotation="0,0,0") -> (Pose) {rotationMode=BMM_Additive, rotationSpace=BCS_BoneSpace, alphaBoolBlend {blendOption=Linear}, boneToModify {boneName=dyn_hair_base}}
   Output Pose(Result=Component To Local(ComponentPose=@n6)) {layerGroup=None}
  ]}
 EventGraph {nodes [Event BlueprintUpdateAnimation()]}
]
variables []
